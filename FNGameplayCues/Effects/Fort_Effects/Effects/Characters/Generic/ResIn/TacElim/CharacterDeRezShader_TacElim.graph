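# Masked two-sided character material for the TacElim de-rez effect, dissolving skinned meshes with glowing blue blocks and grid patterns.
# A sphere mask around ImpactPoint, sized by DissolveGradientPercentage in pre-skinned local space, builds named radial and emissive-band masks.
# Epic quality adds normal-direction explosion, vertex crunch toward the pivot and dithered opacity; lower levels use a panning-noise cutout.
# Both paths fade near the camera and clamp emissive for ray tracing.
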Material /FNGameplayCues/Effects/Fort_Effects/Effects/Characters/Generic/ResIn/TacElim/CharacterDeRezShader_TacElim
graph {
 properties {
  bAutomaticallySetUsageInEditor=false
  bUseMaterialAttributes=true
  bUsedWithClothing=true
  bUsedWithMeshDeformer=true
  bUsedWithMorphTargets=true
  bUsedWithSkeletalMesh=true
  bUsedWithStaticLighting=true
  blendMode=BLEND_Masked
  pixelDepthOffsetMode=PDOM_Legacy
  translucencyLightingMode=TLM_VolumetricDirectional
  twoSided=true
  usedShadingModels=Default Lit
 }
 nodes [
  n1 = LocalPosition() {localOrigin=InstancePreSkinning}
  n2 = VertexInterpolator(VS=@n1.XYZ)
  n3 = SphereMask(
   A=NamedRerouteDeclaration(None=ShaderStageSwitch(PixelShader=@n2.PS, VertexShader=@n1.XYZ)) {name=PReskinnedLocalPostiion, nodeColor=<rgba 1 0.75 0 1>}
   B=TransformPosition(None=VectorParameter() {group=RezOut, parameterName=ImpactPoint, defaultValue=<rgba 0 0 107.339 1>}.RGB) {transformSourceType=TRANSFORMPOSSOURCE_World}
   Radius=LinearInterpolate(A=Constant() {r=950}, Alpha=ScalarParameter() {defaultValue=1, group=SpawnEffect, parameterName=DissolveGradientPercentage}) {constB=0}
  ) {hardnessPercent=0.5}
  NamedRerouteDeclaration(None=Power(Base=@n3) {constExponent=4}) {name=inner radial mask, nodeColor=<rgba 1 0.0234 0 1>}
  n4 = Multiply(A=@n3) {constB=1.5}
  NamedRerouteDeclaration(None=OneMinus(None=Saturate(None=@n4))) {name=Radial Impact Mask, nodeColor=<rgba 0 1 0.8594 1>}
  n5 = Multiply(A=TextureCoordinate()) {constB=3.4424}
  n6 = Multiply(A=OneMinus(None=TextureSample(UVs=@n5) {samplerType=SAMPLERTYPE_LinearColor, texture=/Game/Packages/Fortress_SharedMaterials/Textures/T_Slowfield_GridSquares_01}.R)) {constB=5}
  n7 = NamedRerouteUsage()
  NamedRerouteDeclaration(None=Saturate(None=LinearInterpolate(B=@n6, Alpha=@n7) {desc="To  Opacity Mask"})) {name=EmissiveBand, nodeColor=<rgba 0.5938 1 0 1>}
  n8 = GetMaterialAttributes(None=MaterialFunctionCall!(BaseMaterial, Rim) {materialFunction=/Game/Packages/Fortress_SharedMaterials/Base_Material_Functions/MF_CharacterDerez_BaseBlue}.Result) {attributeGetTypes [#1]}
  n9 = Constant() {r=1}
  n10 = Constant() {r=0.5}
  n11 = ScalarParameter() {defaultValue=1, group=SpawnEffect, parameterName=DissolveGradientPercentage}
  n12 = Multiply(
   A=OneMinus(None=MaterialFunctionCall!(ThresholdValue=@n9, ThresholdRange=@n10, Alpha=@n11, TwoSidedThreshold) {materialFunction=/Engine/Functions/Engine_MaterialFunctions02/Masking/ThresholdWithRange}.Result)
   B=VectorParameter() {parameterName=BlocksColor, defaultValue=<rgba 0 0.1759 0.3 1>}.RGB
  )
  n13 = Power(
   Base=TextureSample(UVs=TextureCoordinate() {uTiling=3, vTiling=3}) {samplerSource=SSM_Wrap_WorldGroupSettings, texture=/Game/Athena/SafeZone/Textures/T_JanusShield3_M}.R
   Exp=ScalarParameter() {defaultValue=0.5, parameterName=BlocksContrast}
  ) {constExponent=0.5}
  n14 = Multiply(A=Add(A=@n8.Emissive Color, B=Multiply(A=@n12, B=@n13)), B=LinearInterpolate(Alpha=NamedRerouteUsage()) {constA=0.4, constB=120})
  n15 = Clamp(None=@n14) {maxDefault=5}
  n16 = SetMaterialAttributes(
   MaterialAttributes=@n8.MaterialAttributes
   Emissive Color=MaterialFunctionCall(Emissive=RayTracingQualitySwitch(Normal=@n14, RayTraced=@n15)) {materialFunction=&MF_FX_EmissiveAdjustments}.Result
  ) {attributeSetTypes [#1]}
  n17 = GetMaterialAttributes(None=@n16) {attributeGetTypes [#2]}
  n18 = ObjectPositionWS()
  n19 = WorldPosition()
  n20 = ScalarParameter() {parameterName=HT_CrunchVerts}
  n21 = NamedRerouteUsage()
  n22 = VertexNormalWS()
  n23 = ScalarParameter() {defaultValue=5, parameterName=ExplodeIntensity}
  n24 = MaterialFunctionCall!(WPO=Multiply(A=OneMinus(None=@n21), B=Multiply(A=@n22, B=@n23))) {materialFunction=/Game/Packages/Fortress_SharedMaterials/Base_Material_Functions/Fortnite_HideElementsWithVertexColor}
  n25 = Add(A=Multiply(A=Subtract(A=@n18, B=@n19.XYZ), B=@n20), B=Multiply(A=OneMinus(None=@n20), B=Add(A=@n17.World Position Offset, B=@n24.Result)))
  n26 = Multiply(
   A=MaterialFunctionCall(In=@n21, Contrast=ScalarParameter() {defaultValue=1, parameterName=OpacityMaskingContrast}) {materialFunction=&CheapContrast}.Result
   B=MaterialFunctionCall(Distance=Constant() {r=65}) {materialFunction=&Near_Camera_Fade}.Result
  )
  n27 = GetMaterialAttributes(None=SetMaterialAttributes!(
    MaterialAttributes=SetMaterialAttributes!(MaterialAttributes=@n17.MaterialAttributes, World Position Offset=@n25) {attributeSetTypes [#2]}
    Opacity Mask=MaterialFunctionCall(Alpha Threshold=@n26) {materialFunction=&DitherTemporalAA}.Result
   ) {attributeSetTypes [#3]}) {attributeGetTypes [#1]}
  n28 = Multiply(A=@n27.Emissive Color, B=LinearInterpolate(Alpha=NamedRerouteUsage()) {constA=0.4, constB=120})
  n29 = RayTracingQualitySwitch(Normal=@n28, RayTraced=Clamp(None=@n28) {maxDefault=5})
  n30 = TextureSample(UVs=Panner(Coordinate=TextureCoordinate() {uTiling=1.5, vTiling=1.5}) {speedY=0.1}) {samplerType=SAMPLERTYPE_LinearColor, texture=/Game/Packages/Fortress_Sky/Materials/Noise32x322_nearestNeighboor}
  n31 = NamedRerouteUsage()
  n32 = Clamp(None=Add(A=Multiply(A=@n30.G, B=@n30.G), B=LinearInterpolate(Alpha=@n31) {constA=-1, constB=2}))
  n33 = Constant() {r=35}
  n34 = SetMaterialAttributes!(
   MaterialAttributes=@n16
   Opacity Mask=Multiply(
    A=Multiply(A=@n32, B=@n31)
    B=MaterialFunctionCall(Distance=@n33) {materialFunction=&Near_Camera_Fade}.Result
   )
  ) {attributeSetTypes [#3]}
  QualitySwitch(
   Default=@n34
   Low=@n34
   High=@n34
   Medium=@n34
   Epic=SetMaterialAttributes!(
    MaterialAttributes=@n27.MaterialAttributes
    Emissive Color=MaterialFunctionCall(Emissive=@n29) {materialFunction=&MF_FX_EmissiveAdjustments}.Result
   ) {attributeSetTypes [#1]}
  )
 ]
}
