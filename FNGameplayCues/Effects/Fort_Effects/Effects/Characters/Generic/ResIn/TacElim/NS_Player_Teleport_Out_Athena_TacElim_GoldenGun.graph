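# Niagara player elimination effect, Golden Gun variant, that bursts the character into gold-orange cubes.
# Two GPU cube emitters fill bone-placed cylinders, differing mainly in lifetime and collision type.
# Cubes outside the impact radius briefly stick to animated bones, with forces, curl noise and collision driving the scatter.
# An impact flash and drifting gold glints accompany them; User impact, force, velocity and headshot inputs steer direction and intensity.
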NiagaraSystem /FNGameplayCues/Effects/Fort_Effects/Effects/Characters/Generic/ResIn/TacElim/NS_Player_Teleport_Out_Athena_TacElim_GoldenGun
dependencies {
 usedDataInterfaces [
  &NiagaraDataInterfaceCurve
  &NiagaraDataInterfaceArrayColor
  &NiagaraDataInterfaceArrayFloat
  &NiagaraDataInterfaceArrayFloat3
  /Script/Niagara.NiagaraDataInterfaceAsyncGpuTrace
  &NiagaraDataInterfaceColorCurve
 ]
 usedDynamicInputs [&SimulationPosition]
 usedModules [
  &self:SystemSpawnScript.SystemScriptSource.SystemScriptGraph.NiagaraNodeAssignment_0.SetVariables_#1
  &SystemState
 ]
}
emitters {
 Dynamic_Cubes {
  simTarget=GPUComputeSim
  emitterSpawnScript [/SetVariables(Emitter.IsHeadshot:NiagaraBool=Linked() {linkedVariable {name=User.Headshot?, type=NiagaraBool, defaultValue=~, description=~}}) {isSetParameters=true, module=SetVariables_#2}]
  emitterUpdateScript [
   &EmitterState(Life Cycle Mode=::NewEnumerator1, Loop Behavior=::NewEnumerator1, Loop Delay=0.1, Loop Duration=0.275, UseLoopDelay=true)
   &SpawnRate(
    SpawnRate=&MakeCustomFloatFromBool(Bool=Linked() {linkedVariable {name=User.Headshot?, type=NiagaraBool, defaultValue=~, description=~}}, True Float=350, False Float=0)
    Use Spawn Probability=false
   )
   &SpawnBurst_Instantaneous(
    Spawn Count=&MakeCustomIntFromBool(
     Bool=Linked() {linkedVariable {name=User.Headshot?, type=NiagaraBool, defaultValue=~, description=~}}
     True INT=&RandomRangeInt(Minimum=25, Maximum=30)
     False INT=&RandomRangeInt(Minimum=15, Maximum=19)
    )
    Use Spawn Probability=false
    UseLoopCountLimit=false
   ) {module=SpawnBurst_Instantaneous005}
   &SpawnBurst_Instantaneous(Spawn Count=100, Spawn Group=1, Use Spawn Probability=false, UseLoopCountLimit=false) {module=SpawnBurst_Instantaneous004}
   &SpawnBurst_Instantaneous(Spawn Count=50, Spawn Group=2, Use Spawn Probability=false, UseLoopCountLimit=false) {module=SpawnBurst_Instantaneous003}
   &SpawnBurst_Instantaneous(Spawn Count=100, Spawn Group=3, Use Spawn Probability=false, UseLoopCountLimit=false) {module=SpawnBurst_Instantaneous002}
   &SpawnBurst_Instantaneous(Spawn Count=220, Spawn Group=4, Use Spawn Probability=false, UseLoopCountLimit=false) {module=SpawnBurst_Instantaneous001}
   &SpawnBurst_Instantaneous(Spawn Count=50, Spawn Group=5, Use Spawn Probability=false, UseLoopCountLimit=false)
   &Timeline(Play=&SetBoolByFloatComparison(A=$Emitter.Age, B=0.255, Comparison Type=::NewEnumerator6))
  ]
  particleSpawnScript [
   &InitializeParticle(
    Color=&MaskLinearColorBySpawnGroup(
     Masked Linear Color=&SelectLinearColorFromArray(
      Array Sampling Mode=::NewEnumerator0
      Direct Array Index=?
      Color Selection Array=DataInterface() {propertyValues {ColorData [<rgba 1 0.6338 0.0447 1>, <rgba 1 0.1667 0 1>, <rgba 1 0.7112 0.3336 1>, <rgba 1 0.4174 0 1>]}}
      Random Seed=0
     )
     Passthrough Linear Color=&RandomRangeLinearColor(Minimum=<rgba 2 0.313 0 1>, Maximum=<rgba 1.5 0.4696 0 1>)
     Spawn Group Mask=&MakeCustomIntFromBool(Bool=Linked() {linkedVariable {name=User.Headshot?, type=NiagaraBool, defaultValue=~, description=~}}, True INT=0, False INT=6)
    )
    Lifetime=&RandomRangeFloat(Minimum=1.998, Maximum=2.32)
    Mass Max=5
    Mass Min=3
    Mass Mode=::NewEnumerator3
    Mesh Scale Mode=::NewEnumerator3
    Mesh Uniform Scale=&Multiply_Float(
     A=&RandomRangeFloat(Minimum=0.75, Maximum=1.5)
     B=&SelectFloatFromArray(
      Direct Array Index=&SpawnGroup()
      Float Selection Array=DataInterface() {propertyValues {FloatData [1.69,1.25,1,2,1.5,0.75]}}
     )
    )
    UsePositionOffset=false
    Write Material Random=true
   )
   &SkeletalMeshLocation(
    Bone / Socket Index=&self:BoneIndexFullSkeleton(
     MaxIndex_SpawnGroup0:NiagaraInt32=0
     MaxIndex_SpawnGroup1:NiagaraInt32=4
     MaxIndex_SpawnGroup2:NiagaraInt32=6
     MinIndex_SpawnGroup0:NiagaraInt32=0
     MinIndex_SpawnGroup1:NiagaraInt32=1
     MinIndex_SpawnGroup2:NiagaraInt32=5
     Spawn Group:NiagaraInt32=&SpawnGroup()
     MinIndex_SpawnGroup3:NiagaraInt32=7
     MinIndex_SpawnGroup4:NiagaraInt32=10
     MinIndex_SpawnGroup5:NiagaraInt32=14
     MaxIndex_SpawnGroup3:NiagaraInt32=9
     MaxIndex_SpawnGroup4:NiagaraInt32=13
     MaxIndex_SpawnGroup5:NiagaraInt32=17
    )
    Bone Sampling Mode=::NewEnumerator4
    Inherit Velocity=false
    Interpolate Mesh Sampling=false
    MaskSpawnGroup=false
    Skeletal Mesh=$User.SkeletalMesh
   )
   &ShapeLocation(
    Additional Quaternion Rotation=true
    Additional Yaw / Pitch / Roll=<0 90 0>
    Cylinder Height=&SelectFloatFromArray(
     Direct Array Index=&SpawnGroup()
     Float Selection Array=DataInterface() {propertyValues {FloatData [18,30,8,20,30,5]}}
    )
    Cylinder Radius=&SelectFloatFromArray(
     Direct Array Index=&SpawnGroup()
     Float Selection Array=DataInterface() {propertyValues {FloatData [6,4,3,10,4,8]}}
    )
    MaskSpawnGroup=false
    Offset=&SelectVectorFromArray(
     Direct Array Index=&SpawnGroup()
     Vector Selection Array=DataInterface() {propertyValues {FloatData [<0 1 0>, <0 2 -10>, <0 1 -3>, <0 0 0>, <0 0 -28>, <0 0 0>]}}
    )
    Offset Coordinate Space=::Local
    Offset Mode=::NewEnumerator0
    Override Seed=false
    Randomness Mode:ENiagaraRandomnessMode=::NewEnumerator1
    Rotation Angle Type:ENiagara_AngleInput=::NewEnumerator0
    Rotation Mode=::NewEnumerator3
    Rotation Quaternion=$Output.SkeletalMeshLocation.SampledRotation
    Shape Primitive=::NewEnumerator1
    Surface Only=false
    UseOwnerScale=true
   )
   &RegionMask(Region Falloff=0, Region Origin:NiagaraPosition=$User.ImpactLocation, Region Radius=55, UseOffset=false, UseRotation=false)
   /SetVariables(Particles.CurrentBoneIndex:NiagaraInt32=$Output.SkeletalMeshLocation.SampledBoneIndex) {isSetParameters=true, module=SetVariables_#3}
   &self:GetBoneMatrix(BoneIndex:NiagaraInt32=$Particles.CurrentBoneIndex, Skeletal Mesh:NiagaraDataInterfaceSkeletalMesh=$User.SkeletalMesh) {module=GetBoneMatrix}
   &PointForce(
    Force Falloff Distance=60
    Force Origin:NiagaraPosition=$User.ImpactLocation
    Force Strength=&RandomRangeFloat(Minimum=490, Maximum=1470)
    Linear Force Vector:Vector3f=&MakeCustomVectorFromBool(
     Bool=Linked() {linkedVariable {name=User.Headshot?, type=NiagaraBool, defaultValue=~, description=~}}
     True Vector=&Add_Vector(A=&Multiply_VectorByFloat(Vector=$User.LinearForceVector, Float=2), B=<0 0 3.25>)
     False Vector=&Multiply_VectorByFloat(
      Vector=&Multiply_Vector(A=&RandomRangeVector(Minimum=<0.75 0.75 0.75>, Maximum=<1.15 1.15 1.15>), B=$User.LinearForceVector)
      Float=15
     )
    )
    Module.Use Falloff=true
    Module.UseLinearForce=true
   )
   &AddVelocity(
    Velocity=&RandomRangeVector(Minimum=<0 0 120>, Maximum=<0 0 350>)
    Velocity Speed Scale=&MakeCustomFloatFromBool(Bool=Linked() {linkedVariable {name=User.Headshot?, type=NiagaraBool, defaultValue=~, description=~}}, True Float=1, False Float=0.25)
   ) {module=AddVelocity001}
   &AddVelocity(
    Cone Axis=&MakeCustomVectorFromBool(
     Bool=&SetBoolByFloatComparison(A=&VectorLength(Vector=$User.PlayerVelocity, Fallback Vector=<0 0 0>), B=0)
     True Vector=$User.PlayerVelocity
     False Vector=$User.LinearForceVector
    )
    Inner Cone Angle=20
    Module.UseConeFalloff=false
    Module.UseVelDistribution=false
    Override Seed=false
    Velocity Mode=::NewEnumerator2
    Velocity Speed=&Multiply_Float(
     A=&RandomRangeFloat(Minimum=300, Maximum=555)
     B=&MakeCustomFloatFromBool(Bool=Linked() {linkedVariable {name=User.Headshot?, type=NiagaraBool, defaultValue=~, description=~}}, True Float=1.6, False Float=1.3)
    )
   )
   &ApplyInitialForces(Positional Force Warmup Time=0)
   &InitialMeshOrientation(
    Rotation=&RandomRangeVector(Minimum=<0 0 0>, Maximum=<1 1 1>)
    Use Orientation Vector=true
    Use Rotation Vector=true
   )
   /Niagara/Modules/Update/Mass/CalculateSizeAndRotationalInertiaByMass(Mass Modulation=5, Module.Manually Enter Density=false)
   &DynamicMaterialParameters(
    Index 0 Param 2=1
    Index 0 Param 4=1
    Module.Index0Param0Write=true
    Module.Index0Param1Write=true
    Module.Index0Param2Write=true
    Module.Index0Param3Write=true
   )
   /SetVariables(Particles.ShouldAdhere:NiagaraBool=&Boolean_NotOperation(Bool=$Particles.RegionMask.IsInsideMask)) {isSetParameters=true, module=SetVariables_#4}
   &Color(Color=&LerpLinearColors(StartColor=$Particles.Color, EndColor=<rgba 1 0.1143 0 1>, LerpFactor=$Output.RegionMask.RegionMask)) {enabled=false}
  ]
  particleUpdateScript [
   &ParticleState()
   &Timeline(
    Initial Start Time=0.3
    Play=&SetBoolByFloatComparison(A=$Particles.Age, B=0.09, Comparison Type=::NewEnumerator6)
   ) {module=Timeline001}
   &self:GetBoneMatrix(BoneIndex:NiagaraInt32=$Particles.CurrentBoneIndex, Skeletal Mesh:NiagaraDataInterfaceSkeletalMesh=$User.SkeletalMesh) {module=GetBoneMatrix001}
   &CurlNoiseForce(
    Noise Frequency=150
    Noise Strength=&Multiply_Float(
     A=500
     B=&MakeCustomFloatFromBool(Bool=Linked() {linkedVariable {name=User.Headshot?, type=NiagaraBool, defaultValue=~, description=~}}, True Float=5, False Float=1)
    )
    Pan Noise=false
   )
   &Drag(
    Drag=&Multiply_Float(
     A=&RandomRangeFloat(Minimum=3.5, Maximum=8)
     B=&FloatFromCurve(FloatCurve=DataInterface() {propertyValues {Curve {keys [
          {interpMode=RCIM_Cubic, leaveTangent=-4.7225, tangentMode=RCTM_Break, time=0, value=1}
          {interpMode=RCIM_Cubic, tangentMode=RCTM_Break, time=0.5, value=0}
         ]}}})
    )
    Use Linear Drag=true
    Use Rotational Drag=true
   )
   &GravityForce()
   &Collision(
    Age Colliding Particles=false
    Collision Enabled=&Boolean_NotOperation(Bool=$Emitter.Timeline.IsPlaying)
    Control Roll On Collision=true
    Friction:NiagaraFloat=&RandomRangeFloat(Minimum=0.2, Maximum=0.65)
    GPU Collision Type=::NewEnumerator2
    Max Number Of Collisions:NiagaraInt32=&MakeCustomIntFromBool(Bool=&SetBoolByFloatComparison(A=$Particles.RegionMask.Float, B=0), True INT=5, False INT=5)
    Maximum Penetration Correction Distance=3
    Mesh Dimensions:Vector3f=$Particles.Scale
    Module.Control Rest State Via Surface Normal=false
    Module.EnableMaxCollisionCount=true
    Module.Randomize Collision Normal Vector=true
    Percentage of Penetration Before Rest=0.2
    Radius Calculation Type=::NewEnumerator1
    Rest State Time Range=1
    Restitution=&RandomRangeFloat(Minimum=0.21, Maximum=0.45)
    ScaleRateOfRotation=true
   ) {module=Collision001}
   &Timeline(Instantaneous Rewind=true, Play=$Output.Collision001.CollisionValid, Rewind Rate=0.1, Timeline End=0.25) {enabled=false, module=Timeline002}
   &SolveForcesAndVelocity(Clamp Velocity=false, Limit Acceleration=false) {module=SolveForcesAndVelocity001}
   &ScaleColor(
    Scale Alpha=&FloatFromCurve(FloatCurve=DataInterface() {propertyValues {Curve {keys [
         {tangentMode=RCTM_User, time=0, value=1}
         {arriveTangent=-0.0405, interpMode=RCIM_Cubic, leaveTangent=-0.0405, tangentMode=RCTM_User, time=0.2, value=1}
         {arriveTangent=-3.4994, interpMode=RCIM_Cubic, leaveTangent=-3.4994, tangentMode=RCTM_User, time=1, value=0}
        ]}}})
    ScaleA=true
    ScaleRGB=false
   )
   &DynamicMaterialParameters(
    Index 0 Param 1=&MakeCustomFloatFromBool(
     Bool=$Particles.Initial.RegionMask.IsInsideMask
     True Float=&FloatFromCurve(FloatCurve=DataInterface() {propertyValues {Curve {keys [
          {arriveTangent=-8.7965, interpMode=RCIM_Cubic, leaveTangent=-8.7965, tangentMode=RCTM_User, time=0, value=1}
          {arriveTangent=-1.8368, interpMode=RCIM_Cubic, leaveTangent=-1.8368, tangentMode=RCTM_User, time=0.22, value=0}
         ]}}})
     False Float=0
    )
    Index 0 Param 2=&FloatFromCurve(FloatCurve=DataInterface() {propertyValues {Curve {keys [
         {interpMode=RCIM_Cubic, tangentMode=RCTM_Break, time=0, value=1}
         {arriveTangent=-2.2159, interpMode=RCIM_Cubic, leaveTangent=-0.0891, tangentMode=RCTM_Break, time=1, value=0}
        ]}}})
    Index 0 Param 3=&Multiply_Float(
     A=3
     B=&FloatFromCurve(
      FloatCurve=DataInterface() {propertyValues {Curve {keys [
          {arriveTangent=9.0725, interpMode=RCIM_Cubic, leaveTangent=9.0725, tangentMode=RCTM_User, time=0, value=0.2}
          {arriveTangent=-2.5249, interpMode=RCIM_Cubic, leaveTangent=-2.5249, tangentMode=RCTM_User, time=0.1408, value=1.5}
          {arriveTangent=-0.0405, interpMode=RCIM_Cubic, leaveTangent=-0.0405, tangentMode=RCTM_User, time=1, value=0.1}
         ]}}}
      CurveIndex=$Particles.Timeline002.NormalizedTime
     )
    )
    Index 0 Param 4=&FloatFromCurve(FloatCurve=DataInterface() {propertyValues {Curve {keys [
         {arriveTangent=10.0923, interpMode=RCIM_Cubic, leaveTangent=10.0923, tangentMode=RCTM_User, time=0, value=0}
         {interpMode=RCIM_Cubic, tangentMode=RCTM_User, time=0.25, value=1}
        ]}}})
    Module.Index0Param0Write=true
    Module.Index0Param1Write=true
    Module.Index0Param2Write=false
    Module.Index0Param3Write=false
   ) {module=DynamicMaterialParameters001}
   &SkeletalMeshLocation(
    Bone / Socket Index=$Particles.CurrentBoneIndex
    Bone Sampling Mode=::NewEnumerator6
    "Inherit Velocity (Scale)":NiagaraFloat=&Multiply_Float(
     A=0.35
     B=&FloatFromCurve(
      FloatCurve=DataInterface() {propertyValues {Curve {keys [
          {interpMode=RCIM_Cubic, tangentMode=RCTM_Break, time=0, value=1}
          {arriveTangent=-5.1891, interpMode=RCIM_Cubic, tangentMode=RCTM_Break, time=0.35, value=0}
         ]}}}
      CurveIndex=$Emitter.Age
     )
    )
    MaskSpawnGroup=false
    Orientation Sampling=::NewEnumerator1
    Sampled Position Offset Coordinate Space:ENiagaraSkelMeshTransforms=::NewEnumerator1
    Skeletal Mesh=$User.SkeletalMesh
    Velocity Sampling=::NewEnumerator0
    VelocitySpeedLimit=true
    Write To Particle Position:NiagaraBool=$Emitter.Timeline.IsPlaying
   ) {module=SkeletalMeshLocation001}
   &ShapeLocation(
    Additional Quaternion Rotation=true
    Additional Yaw / Pitch / Roll=<0 90 0>
    Cylinder Height=&SelectFloatFromArray(
     Direct Array Index=&SpawnGroup()
     Float Selection Array=DataInterface() {propertyValues {FloatData [18,30,8,20,30,5]}}
    )
    Cylinder Radius=&SelectFloatFromArray(
     Direct Array Index=&SpawnGroup()
     Float Selection Array=DataInterface() {propertyValues {FloatData [6,4,3,10,4,8]}}
    )
    MaskSpawnGroup=false
    Offset=&SelectVectorFromArray(
     Direct Array Index=&SpawnGroup()
     Vector Selection Array=DataInterface() {propertyValues {FloatData [<0 1 0>, <0 2 -10>, <0 1 -3>, <0 0 0>, <0 0 -28>, <0 0 0>]}}
     Random Seed=0
    )
    Offset Coordinate Space=::Local
    Offset Mode=::NewEnumerator0
    Override Seed=false
    Rotation Angle Type:ENiagara_AngleInput=::NewEnumerator0
    Rotation Mode=::NewEnumerator3
    Rotation Quaternion=$Particles.SkeletalMeshLocation001.SampledRotation
    Shape Origin=$StackContext.Position
    Shape Primitive=::NewEnumerator1
    Surface Only=false
    UseOwnerScale=true
   ) {module=ShapeLocation002}
   /SetVariables(Particles.Position=/Niagara/DynamicInputs/TypeConversions/MakeCustomPositionFromBool(
     Bool:NiagaraBool=&Boolean_AndOperation(Bool1=$Particles.ShouldAdhere, Bool2=$Emitter.Timeline.IsPlaying)
     True Position:NiagaraPosition=$Particles.ShapeLocation002.ShapePosition
     False Position:NiagaraPosition=$Output.SolveForcesAndVelocity001.Position
    )) {isSetParameters=true, module=SetVariables_#5}
   &self:CubeCollision_01(
    NumberOfCollisions:NiagaraInt32=$Particles.Number Of Collisions
    Rate Of Change=0.25
    SurfaceNormal:Vector3f=$Output.Collision001.CombinedCollisionNormal
   ) {module=CubeCollision_01}
  ]
  rendererClasses [&NiagaraMeshRendererProperties]
  renderers [{rendererClass=&NiagaraMeshRendererProperties, rendererIndex=0}]
 }
 Dynamic_Cubes_Med_High {
  simTarget=GPUComputeSim
  emitterSpawnScript [/SetVariables(Emitter.IsHeadshot:NiagaraBool=Linked() {linkedVariable {name=User.Headshot?, type=NiagaraBool, defaultValue=~, description=~}}) {isSetParameters=true, module=SetVariables_#2}]
  emitterUpdateScript [
   &EmitterState(Life Cycle Mode=::NewEnumerator1, Loop Behavior=::NewEnumerator1, Loop Delay=0.1, Loop Duration=0.275, UseLoopDelay=true)
   &SpawnRate(
    SpawnRate=&MakeCustomFloatFromBool(Bool=Linked() {linkedVariable {name=User.Headshot?, type=NiagaraBool, defaultValue=~, description=~}}, True Float=350, False Float=0)
    Use Spawn Probability=false
   )
   &SpawnBurst_Instantaneous(
    Spawn Count=&MakeCustomIntFromBool(
     Bool=Linked() {linkedVariable {name=User.Headshot?, type=NiagaraBool, defaultValue=~, description=~}}
     True INT=&RandomRangeInt(Minimum=25, Maximum=30)
     False INT=&RandomRangeInt(Minimum=15, Maximum=19)
    )
    Use Spawn Probability=false
    UseLoopCountLimit=false
   ) {module=SpawnBurst_Instantaneous005}
   &SpawnBurst_Instantaneous(Spawn Count=100, Spawn Group=1, Use Spawn Probability=false, UseLoopCountLimit=false) {module=SpawnBurst_Instantaneous004}
   &SpawnBurst_Instantaneous(Spawn Count=50, Spawn Group=2, Use Spawn Probability=false, UseLoopCountLimit=false) {module=SpawnBurst_Instantaneous003}
   &SpawnBurst_Instantaneous(Spawn Count=100, Spawn Group=3, Use Spawn Probability=false, UseLoopCountLimit=false) {module=SpawnBurst_Instantaneous002}
   &SpawnBurst_Instantaneous(Spawn Count=220, Spawn Group=4, Use Spawn Probability=false, UseLoopCountLimit=false) {module=SpawnBurst_Instantaneous001}
   &SpawnBurst_Instantaneous(Spawn Count=50, Spawn Group=5, Use Spawn Probability=false, UseLoopCountLimit=false)
   &Timeline(Play=&SetBoolByFloatComparison(A=$Emitter.Age, B=0.255, Comparison Type=::NewEnumerator6))
  ]
  particleSpawnScript [
   &InitializeParticle(
    Color=&MaskLinearColorBySpawnGroup(
     Masked Linear Color=&SelectLinearColorFromArray(
      Array Sampling Mode=::NewEnumerator0
      Direct Array Index=?
      Color Selection Array=DataInterface() {propertyValues {ColorData [<rgba 1 0.6338 0.0447 1>, <rgba 1 0.1667 0 1>, <rgba 1 0.7112 0.3336 1>, <rgba 1 0.4174 0 1>]}}
      Random Seed=0
     )
     Passthrough Linear Color=&RandomRangeLinearColor(Minimum=<rgba 2 0.313 0 1>, Maximum=<rgba 1.5 0.4696 0 1>)
     Spawn Group Mask=&MakeCustomIntFromBool(Bool=Linked() {linkedVariable {name=User.Headshot?, type=NiagaraBool, defaultValue=~, description=~}}, True INT=0, False INT=6)
    )
    Lifetime=&RandomRangeFloat(Minimum=0.85, Maximum=1.4)
    Mass Max=5
    Mass Min=3
    Mass Mode=::NewEnumerator3
    Mesh Scale Mode=::NewEnumerator3
    Mesh Uniform Scale=&Multiply_Float(
     A=&RandomRangeFloat(Minimum=0.75, Maximum=1.5)
     B=&SelectFloatFromArray(
      Direct Array Index=&SpawnGroup()
      Float Selection Array=DataInterface() {propertyValues {FloatData [1.69,1.25,1,2,1.5,0.75]}}
     )
    )
    UsePositionOffset=false
    Write Material Random=true
   )
   &SkeletalMeshLocation(
    Bone / Socket Index=&self:BoneIndexFullSkeleton(
     MaxIndex_SpawnGroup0:NiagaraInt32=0
     MaxIndex_SpawnGroup1:NiagaraInt32=4
     MaxIndex_SpawnGroup2:NiagaraInt32=6
     MinIndex_SpawnGroup0:NiagaraInt32=0
     MinIndex_SpawnGroup1:NiagaraInt32=1
     MinIndex_SpawnGroup2:NiagaraInt32=5
     Spawn Group:NiagaraInt32=&SpawnGroup()
     MinIndex_SpawnGroup3:NiagaraInt32=7
     MinIndex_SpawnGroup4:NiagaraInt32=10
     MinIndex_SpawnGroup5:NiagaraInt32=14
     MaxIndex_SpawnGroup3:NiagaraInt32=9
     MaxIndex_SpawnGroup4:NiagaraInt32=13
     MaxIndex_SpawnGroup5:NiagaraInt32=17
    )
    Bone Sampling Mode=::NewEnumerator4
    Inherit Velocity=false
    Interpolate Mesh Sampling=false
    MaskSpawnGroup=false
    Skeletal Mesh=$User.SkeletalMesh
   )
   &ShapeLocation(
    Additional Quaternion Rotation=true
    Additional Yaw / Pitch / Roll=<0 90 0>
    Cylinder Height=&SelectFloatFromArray(
     Direct Array Index=&SpawnGroup()
     Float Selection Array=DataInterface() {propertyValues {FloatData [18,30,8,20,30,5]}}
    )
    Cylinder Radius=&SelectFloatFromArray(
     Direct Array Index=&SpawnGroup()
     Float Selection Array=DataInterface() {propertyValues {FloatData [6,4,3,10,4,8]}}
    )
    MaskSpawnGroup=false
    Offset=&SelectVectorFromArray(
     Direct Array Index=&SpawnGroup()
     Vector Selection Array=DataInterface() {propertyValues {FloatData [<0 1 0>, <0 2 -10>, <0 1 -3>, <0 0 0>, <0 0 -28>, <0 0 0>]}}
    )
    Offset Coordinate Space=::Local
    Offset Mode=::NewEnumerator0
    Override Seed=false
    Randomness Mode:ENiagaraRandomnessMode=::NewEnumerator1
    Rotation Angle Type:ENiagara_AngleInput=::NewEnumerator0
    Rotation Mode=::NewEnumerator3
    Rotation Quaternion=$Output.SkeletalMeshLocation.SampledRotation
    Shape Primitive=::NewEnumerator1
    Surface Only=false
    UseOwnerScale=true
   )
   &RegionMask(Region Falloff=0, Region Origin:NiagaraPosition=$User.ImpactLocation, Region Radius=55, UseOffset=false, UseRotation=false)
   /SetVariables(Particles.CurrentBoneIndex:NiagaraInt32=$Output.SkeletalMeshLocation.SampledBoneIndex) {isSetParameters=true, module=SetVariables_#3}
   &self:GetBoneMatrix(BoneIndex:NiagaraInt32=$Particles.CurrentBoneIndex, Skeletal Mesh:NiagaraDataInterfaceSkeletalMesh=$User.SkeletalMesh) {module=GetBoneMatrix}
   &PointForce(
    Force Falloff Distance=60
    Force Origin:NiagaraPosition=$User.ImpactLocation
    Force Strength=&RandomRangeFloat(Minimum=490, Maximum=1470)
    Linear Force Vector:Vector3f=&MakeCustomVectorFromBool(
     Bool=Linked() {linkedVariable {name=User.Headshot?, type=NiagaraBool, defaultValue=~, description=~}}
     True Vector=&Add_Vector(A=&Multiply_VectorByFloat(Vector=$User.LinearForceVector, Float=2), B=<0 0 3.25>)
     False Vector=&Multiply_VectorByFloat(
      Vector=&Multiply_Vector(A=&RandomRangeVector(Minimum=<0.75 0.75 0.75>, Maximum=<1.15 1.15 1.15>), B=$User.LinearForceVector)
      Float=15
     )
    )
    Module.Use Falloff=true
    Module.UseLinearForce=true
   )
   &AddVelocity(
    Velocity=&RandomRangeVector(Minimum=<0 0 120>, Maximum=<0 0 350>)
    Velocity Speed Scale=&MakeCustomFloatFromBool(Bool=Linked() {linkedVariable {name=User.Headshot?, type=NiagaraBool, defaultValue=~, description=~}}, True Float=1, False Float=0.25)
   ) {module=AddVelocity001}
   &AddVelocity(
    Cone Axis=&MakeCustomVectorFromBool(
     Bool=&SetBoolByFloatComparison(A=&VectorLength(Vector=$User.PlayerVelocity, Fallback Vector=<0 0 0>), B=0)
     True Vector=$User.PlayerVelocity
     False Vector=$User.LinearForceVector
    )
    Inner Cone Angle=20
    Module.UseConeFalloff=false
    Module.UseVelDistribution=false
    Override Seed=false
    Velocity Mode=::NewEnumerator2
    Velocity Speed=&Multiply_Float(
     A=&RandomRangeFloat(Minimum=300, Maximum=555)
     B=&MakeCustomFloatFromBool(Bool=Linked() {linkedVariable {name=User.Headshot?, type=NiagaraBool, defaultValue=~, description=~}}, True Float=1.6, False Float=1.3)
    )
   )
   &ApplyInitialForces(Positional Force Warmup Time=0)
   &InitialMeshOrientation(
    Rotation=&RandomRangeVector(Minimum=<0 0 0>, Maximum=<1 1 1>)
    Use Orientation Vector=true
    Use Rotation Vector=true
   )
   &CalculateSizeAndRotationalInertiaByMass(Mass Modulation=5, Module.Manually Enter Density=false)
   &DynamicMaterialParameters(
    Index 0 Param 2=1
    Index 0 Param 4=1
    Module.Index0Param0Write=true
    Module.Index0Param1Write=true
    Module.Index0Param2Write=true
    Module.Index0Param3Write=true
   )
   /SetVariables(Particles.ShouldAdhere:NiagaraBool=&Boolean_NotOperation(Bool=$Particles.RegionMask.IsInsideMask)) {isSetParameters=true, module=SetVariables_#4}
   &Color(Color=&LerpLinearColors(StartColor=$Particles.Color, EndColor=<rgba 1 0.1143 0 1>, LerpFactor=$Output.RegionMask.RegionMask)) {enabled=false}
  ]
  particleUpdateScript [
   &ParticleState()
   &Timeline(
    Initial Start Time=0.3
    Play=&SetBoolByFloatComparison(A=$Particles.Age, B=0.09, Comparison Type=::NewEnumerator6)
   ) {module=Timeline001}
   &self:GetBoneMatrix(BoneIndex:NiagaraInt32=$Particles.CurrentBoneIndex, Skeletal Mesh:NiagaraDataInterfaceSkeletalMesh=$User.SkeletalMesh) {module=GetBoneMatrix001}
   &CurlNoiseForce(
    Noise Frequency=150
    Noise Strength=&Multiply_Float(
     A=500
     B=&MakeCustomFloatFromBool(Bool=Linked() {linkedVariable {name=User.Headshot?, type=NiagaraBool, defaultValue=~, description=~}}, True Float=5, False Float=1)
    )
    Pan Noise=false
   )
   &Drag(
    Drag=&Multiply_Float(
     A=&RandomRangeFloat(Minimum=3.5, Maximum=8)
     B=&FloatFromCurve(FloatCurve=DataInterface() {propertyValues {Curve {keys [
          {interpMode=RCIM_Cubic, leaveTangent=-4.7225, tangentMode=RCTM_Break, time=0, value=1}
          {interpMode=RCIM_Cubic, tangentMode=RCTM_Break, time=0.5, value=0}
         ]}}})
    )
    Use Linear Drag=true
    Use Rotational Drag=true
   )
   &GravityForce()
   &Collision(
    Age Colliding Particles=false
    Analytical Collision Plane Position 1:Vector3f=<0 0 0>
    Collision Enabled=&Boolean_NotOperation(Bool=$Emitter.Timeline.IsPlaying)
    Control Roll On Collision=true
    Friction:NiagaraFloat=&RandomRangeFloat(Minimum=0.2, Maximum=0.65)
    GPU Collision Type=::NewEnumerator3
    Max Number Of Collisions=3
    Maximum Penetration Correction Distance=3
    Mesh Dimensions:Vector3f=$Particles.Scale
    Module.Control Rest State Via Surface Normal=false
    Module.EnableMaxCollisionCount=true
    Module.Randomize Collision Normal Vector=true
    Percentage of Penetration Before Rest=0.2
    Radius Calculation Type=::NewEnumerator1
    Rest State Time Range=1
    Restitution=&RandomRangeFloat(Minimum=0.21, Maximum=0.45)
    ScaleRateOfRotation=true
   ) {module=Collision001}
   &Timeline(Instantaneous Rewind=true, Play=$Output.Collision001.CollisionValid, Rewind Rate=0.1, Timeline End=0.25) {enabled=false, module=Timeline002}
   &SolveForcesAndVelocity(Clamp Velocity=false, Limit Acceleration=false) {module=SolveForcesAndVelocity001}
   &ScaleColor(
    Scale Alpha=&FloatFromCurve(FloatCurve=DataInterface() {propertyValues {Curve {keys [
         {tangentMode=RCTM_User, time=0, value=1}
         {arriveTangent=-0.0405, interpMode=RCIM_Cubic, leaveTangent=-0.0405, tangentMode=RCTM_User, time=0.2, value=1}
         {arriveTangent=-3.4994, interpMode=RCIM_Cubic, leaveTangent=-3.4994, tangentMode=RCTM_User, time=1, value=0}
        ]}}})
    ScaleA=true
    ScaleRGB=false
   )
   &DynamicMaterialParameters(
    Index 0 Param 1=&MakeCustomFloatFromBool(
     Bool=$Particles.Initial.RegionMask.IsInsideMask
     True Float=&FloatFromCurve(FloatCurve=DataInterface() {propertyValues {Curve {keys [
          {arriveTangent=-8.7965, interpMode=RCIM_Cubic, leaveTangent=-8.7965, tangentMode=RCTM_User, time=0, value=1}
          {arriveTangent=-1.8368, interpMode=RCIM_Cubic, leaveTangent=-1.8368, tangentMode=RCTM_User, time=0.22, value=0}
         ]}}})
     False Float=0
    )
    Index 0 Param 2=&FloatFromCurve(FloatCurve=DataInterface() {propertyValues {Curve {keys [
         {interpMode=RCIM_Cubic, tangentMode=RCTM_Break, time=0, value=1}
         {arriveTangent=-2.2159, interpMode=RCIM_Cubic, leaveTangent=-0.0891, tangentMode=RCTM_Break, time=1, value=0}
        ]}}})
    Index 0 Param 3=&Multiply_Float(
     A=3
     B=&FloatFromCurve(
      FloatCurve=DataInterface() {propertyValues {Curve {keys [
          {arriveTangent=9.0725, interpMode=RCIM_Cubic, leaveTangent=9.0725, tangentMode=RCTM_User, time=0, value=0.2}
          {arriveTangent=-2.5249, interpMode=RCIM_Cubic, leaveTangent=-2.5249, tangentMode=RCTM_User, time=0.1408, value=1.5}
          {arriveTangent=-0.0405, interpMode=RCIM_Cubic, leaveTangent=-0.0405, tangentMode=RCTM_User, time=1, value=0.1}
         ]}}}
      CurveIndex=$Particles.Timeline002.NormalizedTime
     )
    )
    Index 0 Param 4=&FloatFromCurve(FloatCurve=DataInterface() {propertyValues {Curve {keys [
         {arriveTangent=10.0923, interpMode=RCIM_Cubic, leaveTangent=10.0923, tangentMode=RCTM_User, time=0, value=0}
         {interpMode=RCIM_Cubic, tangentMode=RCTM_User, time=0.25, value=1}
        ]}}})
    Module.Index0Param0Write=true
    Module.Index0Param1Write=true
    Module.Index0Param2Write=false
    Module.Index0Param3Write=false
   ) {module=DynamicMaterialParameters001}
   &SkeletalMeshLocation(
    Bone / Socket Index=$Particles.CurrentBoneIndex
    Bone Sampling Mode=::NewEnumerator6
    "Inherit Velocity (Scale)":NiagaraFloat=&Multiply_Float(
     A=0.35
     B=&FloatFromCurve(
      FloatCurve=DataInterface() {propertyValues {Curve {keys [
          {interpMode=RCIM_Cubic, tangentMode=RCTM_Break, time=0, value=1}
          {arriveTangent=-5.1891, interpMode=RCIM_Cubic, tangentMode=RCTM_Break, time=0.35, value=0}
         ]}}}
      CurveIndex=$Emitter.Age
     )
    )
    MaskSpawnGroup=false
    Orientation Sampling=::NewEnumerator1
    Skeletal Mesh=$User.SkeletalMesh
    Velocity Sampling=::NewEnumerator0
    VelocitySpeedLimit=true
    Write To Particle Position:NiagaraBool=$Emitter.Timeline.IsPlaying
   ) {module=SkeletalMeshLocation001}
   &ShapeLocation(
    Additional Quaternion Rotation=true
    Additional Yaw / Pitch / Roll=<0 90 0>
    Cylinder Height=&SelectFloatFromArray(
     Direct Array Index=&SpawnGroup()
     Float Selection Array=DataInterface() {propertyValues {FloatData [18,30,8,20,30,5]}}
    )
    Cylinder Radius=&SelectFloatFromArray(
     Direct Array Index=&SpawnGroup()
     Float Selection Array=DataInterface() {propertyValues {FloatData [6,4,3,10,4,8]}}
    )
    MaskSpawnGroup=false
    Offset=&SelectVectorFromArray(
     Direct Array Index=&SpawnGroup()
     Vector Selection Array=DataInterface() {propertyValues {FloatData [<0 1 0>, <0 2 -10>, <0 1 -3>, <0 0 0>, <0 0 -28>, <0 0 0>]}}
     Random Seed=0
    )
    Offset Coordinate Space=::Local
    Offset Mode=::NewEnumerator0
    Override Seed=false
    Rotation Angle Type:ENiagara_AngleInput=::NewEnumerator0
    Rotation Mode=::NewEnumerator3
    Rotation Quaternion=$Particles.SkeletalMeshLocation001.SampledRotation
    Shape Origin=$StackContext.Position
    Shape Primitive=::NewEnumerator1
    Surface Only=false
    UseOwnerScale=true
   ) {module=ShapeLocation002}
   /SetVariables(Particles.Position=&MakeCustomPositionFromBool(
     Bool:NiagaraBool=&Boolean_AndOperation(Bool1=$Particles.ShouldAdhere, Bool2=$Emitter.Timeline.IsPlaying)
     True Position:NiagaraPosition=$Particles.ShapeLocation002.ShapePosition
     False Position:NiagaraPosition=$Output.SolveForcesAndVelocity001.Position
    )) {isSetParameters=true, module=SetVariables_#5}
   &self:CubeCollision_01(
    NumberOfCollisions:NiagaraInt32=$Particles.Number Of Collisions
    Rate Of Change=0.25
    SurfaceNormal:Vector3f=$Output.Collision001.CombinedCollisionNormal
   ) {module=CubeCollision_01}
  ]
  rendererClasses [&NiagaraMeshRendererProperties]
  renderers [{rendererClass=&NiagaraMeshRendererProperties, rendererIndex=0}]
 }
 Flash {
  emitterUpdateScript [
   &EmitterState()
   &SpawnBurst_Instantaneous(Spawn Count=3, Use Spawn Probability=false, UseLoopCountLimit=false)
  ]
  particleSpawnScript [
   &InitializeParticle(
    Color=<rgba 1 0.42 0 1>
    Lifetime Max=0.3
    Lifetime Min=0.235
    Lifetime Mode=::NewEnumerator1
    Mesh Scale=&Multiply_VectorByFloat(
     Vector=&MakeVector(
      X=&RandomRangeFloat(Minimum=0.2, Maximum=0.2)
      Y=&RandomRangeFloat(Minimum=0.2, Maximum=0.3)
      Z=&RandomRangeFloat(Minimum=0.2, Maximum=0.3)
     )
     Float=1
    )
    Mesh Scale Mode=::NewEnumerator4
    Position=$Particles.Position
    Position Mode=::NewEnumerator1
    Sprite Rotation Mode=::NewEnumerator2
    Sprite Size Mode=::NewEnumerator3
    Uniform Sprite Size=&Multiply_Float(
     A=&Multiply_Float(A=&RandomRangeFloat(Minimum=125, Maximum=175), B=$Particles.AmmoTypeScale)
     B=1
    )
    UsePositionOffset=false
    Write Material Random=true
   )
   &ShapeLocation(
    MaskSpawnGroup=false
    Non Uniform Scale=<1 1 1.5>
    Offset Coordinate Space=::World
    Offset Mode=::NewEnumerator0
    Override Seed=false
    Shape Origin=$User.ImpactLocation
    Sphere Radius=1
    UseOwnerScale=true
   )
   &InitialMeshOrientation(
    Mesh Orientation Mode=::NewEnumerator2
    Orientation Vector=$User.LinearForceVector
    Rotation=&RandomRangeVector(Minimum=<-0.1 -0.2 0.4>, Maximum=<0.1 0.2 0.6>)
    Use Rotation Vector=true
   )
   &CameraOffset(Camera Offset Amount=20)
  ]
  particleUpdateScript [
   &ParticleState()
   &Color(
    Color=&MakeLinearColorFromVectorAndFloat(
     "Vector (RGB)"=&MakeVectorFromLinearColorRGB(Color=&ColorFromCurve(ColorCurve=DataInterface() {propertyValues {
         AlphaCurve {keys []}
         BlueCurve {keys [{arriveTangent=-1.4036, interpMode=RCIM_Cubic, time=0.1352, value=0}, {interpMode=RCIM_Cubic, leaveTangent=-1.4036, time=1, value=0}]}
         GreenCurve {keys [{arriveTangent=-5.857, interpMode=RCIM_Cubic, time=0.1352, value=6.127}, {interpMode=RCIM_Cubic, leaveTangent=-5.857, time=1, value=0.1916}]}
         RedCurve {keys [{arriveTangent=-14.4701, interpMode=RCIM_Cubic, time=0.1352, value=50}, {interpMode=RCIM_Cubic, leaveTangent=-14.4701, time=1, value=0.5625}]}
        }}))
     "Float (Alpha)"=&FloatFromCurve(FloatCurve=DataInterface() {propertyValues {Curve {keys [{arriveTangent=-2.7272, interpMode=RCIM_Cubic, time=0, value=1}, {interpMode=RCIM_Cubic, tangentMode=RCTM_User, time=1, value=0}]}}})
    )
    Scale Color=&VectorFromFloat(Value=&FloatFromCurve(FloatCurve=DataInterface() {propertyValues {Curve {keys [{interpMode=RCIM_Cubic, tangentMode=RCTM_User, time=0, value=1}, {interpMode=RCIM_Cubic, tangentMode=RCTM_User, time=1, value=0}]}}}))
   )
  ]
  rendererClasses [&NiagaraMeshRendererProperties]
  renderers [{rendererClass=&NiagaraMeshRendererProperties, rendererIndex=0}]
 }
 Gold_Glints {
  emitterUpdateScript [
   &EmitterState(Life Cycle Mode=::NewEnumerator1, Loop Behavior=::NewEnumerator1, Loop Duration=0.275, UseLoopDelay=false)
   &SpawnRate(SpawnRate=150, Use Spawn Probability=false)
  ]
  particleSpawnScript [
   &InitializeParticle(
    Color=&UniformRangedLinearColor(Minimum=<rgba 10 5.7319 2.2396 0.4>, Maximum=<rgba 5 2.5 0 1>)
    Lifetime=&UniformRangedFloat(Minimum=1.5, Maximum=2.2)
    Mass=5
    Mass Mode=::NewEnumerator1
    Sprite Size Mode=::NewEnumerator1
    Uniform Sprite Size Max=50
    Uniform Sprite Size Min=40
    UsePositionOffset=false
    Write Material Random=true
   ) {module=InitializeParticle001}
   &SkeletalMeshLocation(Inherit Velocity=false, MaskSpawnGroup=false, Module.Override Randomness=false, Skeletal Mesh=$User.SkeletalMesh) {enabled=false}
   &SphereLocation(
    MaskSpawnGroup=false
    Override Seed=false
    Sphere Coordinate Space=::World
    Sphere Origin:NiagaraPosition=$User.ImpactLocation
    Sphere Radius=5
    Surface Only=false
    UseNonUniformScale=false
    UseOffset=false
    UseSphereOrientation=false
   )
   &CameraOffset(Camera Offset Amount=10)
   &PointForce(
    Force Falloff Distance=40
    Force Origin:NiagaraPosition=$User.ImpactLocation
    Force Strength=&RandomRangeFloat(Minimum=245, Maximum=735)
    Linear Force Vector:Vector3f=&MakeCustomVectorFromBool(
     Bool=Linked() {linkedVariable {name=User.Headshot?, type=NiagaraBool, defaultValue=~, description=~}}
     True Vector=&Add_Vector(A=&Multiply_VectorByFloat(Vector=$User.LinearForceVector, Float=2), B=<0 0 3.25>)
     False Vector=&Multiply_VectorByFloat(
      Vector=&Multiply_Vector(A=&RandomRangeVector(Minimum=<0.75 0.75 1>, Maximum=<1.15 1.15 1.15>), B=$User.LinearForceVector)
      Float=&RandomRangeFloat(Minimum=10, Maximum=15)
     )
    )
    Module.Use Falloff=true
    Module.UseLinearForce=true
   ) {enabled=false}
   &AddVelocity(Velocity=&RandomRangeVector(Minimum=<0 0 120>, Maximum=<0 0 350>)) {module=AddVelocity001}
   &AddVelocity(
    Cone Angle=30
    Cone Axis=&MakeCustomVectorFromBool(
     Bool=&SetBoolByFloatComparison(A=&VectorLength(Vector=$User.PlayerVelocity, Fallback Vector=<0 0 0>), B=0)
     True Vector=$User.PlayerVelocity
     False Vector=$User.LinearForceVector
    )
    Inner Cone Angle=20
    Module.UseConeFalloff=false
    Module.UseVelDistribution=false
    Override Seed=false
    Velocity Mode=::NewEnumerator2
    Velocity Speed=&Multiply_Float(
     A=&RandomRangeFloat(Minimum=222, Maximum=666)
     B=&MakeCustomFloatFromBool(Bool=Linked() {linkedVariable {name=User.Headshot?, type=NiagaraBool, defaultValue=~, description=~}}, True Float=1.6, False Float=1.3)
    )
   ) {module=AddVelocity002}
   &ApplyInitialForces()
  ]
  particleUpdateScript [
   &ParticleState()
   &GravityForce(Gravity=&RandomRangeVector(Minimum=<0 0 -100>, Maximum=<0 0 -200>))
   &Drag(
    Drag=&FloatFromCurve(
     FloatCurve=DataInterface() {propertyValues {Curve {
        defaultValue=1
        keys [
         {arriveTangent=1.5741, interpMode=RCIM_Cubic, leaveTangent=1.5741, tangentMode=RCTM_User, time=0, value=0.4}
         {tangentMode=RCTM_User, time=1, value=1}
        ]
       }}}
     Scale Curve=&RandomRangeFloat(Minimum=3, Maximum=8)
    )
    Use Linear Drag=true
    Use Rotational Drag=false
   ) {module=Drag001}
   &CurlNoiseForce(
    Bias Noise=false
    Noise Strength=&FloatFromCurve(
     FloatCurve=DataInterface() {propertyValues {Curve {
        defaultValue=1
        keys [{interpMode=RCIM_Cubic, tangentMode=RCTM_User, time=0, value=0}, {interpMode=RCIM_Cubic, tangentMode=RCTM_User, time=1, value=1}]
       }}}
     Scale Curve=300
    )
    Pan Noise=false
   )
   &SolveForcesAndVelocity(Clamp Velocity=false, Limit Acceleration=false)
   &ScaleSpriteSize(Uniform Curve Sprite Scale=DataInterface() {propertyValues {Curve {
       defaultValue=1
       keys [
        {arriveTangent=1.3768, interpMode=RCIM_Cubic, leaveTangent=1.3768, tangentMode=RCTM_User, time=0, value=0.8}
        {interpMode=RCIM_Cubic, tangentMode=RCTM_User, time=0.3, value=1}
        {tangentMode=RCTM_User, time=1, value=0}
       ]
      }}})
   &ScaleColor(
    Scale Alpha=&FloatFromCurve(FloatCurve=DataInterface() {propertyValues {Curve {keys [<at 0 1 RCIM_Cubic>, {arriveTangent=-2.9021, tangentMode=RCTM_User, time=1, value=0}]}}})
    ScaleA=true
    ScaleRGB=false
   ) {module=ScaleColor001}
   &DynamicMaterialParameters(
    Index 0 Param 2=1
    Index 0 Param 3=3
    Module.Index0Float0Write=true
    Module.Index0Float1Write=false
    Module.Index0Float2Write=true
    Module.Index0Float3Write=false
   )
   &ScaleColor(
    Scale Alpha=&FloatFromCurve(FloatCurve=DataInterface() {propertyValues {Curve {
        defaultValue=1
        keys [
         {arriveTangent=2.6539, interpMode=RCIM_Cubic, leaveTangent=2.6539, tangentMode=RCTM_User, time=0, value=0.6}
         {interpMode=RCIM_Cubic, tangentMode=RCTM_User, time=0.2, value=1}
         {arriveTangent=0.0043, interpMode=RCIM_Cubic, leaveTangent=0.0043, tangentMode=RCTM_User, time=0.5, value=0.55}
         {arriveTangent=-0.0082, interpMode=RCIM_Cubic, leaveTangent=-0.0082, tangentMode=RCTM_User, time=0.7, value=0.8}
         {tangentMode=RCTM_User, time=1, value=0}
        ]
       }}})
    Scale RGB=&VectorFromFloat(Value=&FloatFromCurve(FloatCurve=DataInterface() {propertyValues {Curve {
         defaultValue=0
         keys [{interpMode=RCIM_Cubic, tangentMode=RCTM_User, time=0, value=1}, {interpMode=RCIM_Cubic, tangentMode=RCTM_User, time=1, value=0}]
        }}}))
    ScaleA=true
    ScaleRGB=true
   )
   /SetVariables(Particles.Glow_Color:LinearColor=&MakeLinearColorFromVectorAndFloat("Vector (RGB)"=&MakeVectorFromLinearColorRGB(Color=$Particles.Color), "Float (Alpha)"=0.01)) {isSetParameters=true}
   /SetVariables(Particles.Glow_Size:Vector2f=&Multiply_Vector2D(A=$Particles.SpriteSize, B=<0.2 0.2>)) {isSetParameters=true}
  ]
  rendererClasses [&NiagaraSpriteRendererProperties]
  renderers [
   {rendererClass=&NiagaraSpriteRendererProperties}
   {rendererClass=&NiagaraSpriteRendererProperties, rendererIndex=1}
  ]
 }
 ImpactSquib {
  simTarget=GPUComputeSim
  emitterSpawnScript [/SetVariables(Emitter.IsHeadshot:NiagaraBool=Linked() {linkedVariable {name=User.Headshot?, type=NiagaraBool, defaultValue=~, description=~}}) {isSetParameters=true, module=SetVariables_#2}]
  emitterUpdateScript [
   &EmitterState(Life Cycle Mode=::NewEnumerator1, Loop Behavior=::NewEnumerator1, Loop Duration=0.2, UseLoopDelay=false)
   &SpawnBurst_Instantaneous(
    Spawn Count=&RandomRangeInt(Minimum=4, Maximum=6)
    Spawn Group:NiagaraInt32=&MakeCustomIntFromBool(Bool=Linked() {linkedVariable {name=User.Headshot?, type=NiagaraBool, defaultValue=~, description=~}}, True INT=3, False INT=1)
    Use Spawn Probability=false
    UseLoopCountLimit=false
   ) {module=SpawnBurst_Instantaneous001}
   &SpawnBurst_Instantaneous(
    Spawn Count=&RandomRangeInt(Minimum=13, Maximum=17)
    Spawn Group:NiagaraInt32=&MakeCustomIntFromBool(Bool=Linked() {linkedVariable {name=User.Headshot?, type=NiagaraBool, defaultValue=~, description=~}}, True INT=2, False INT=0)
    Use Spawn Probability=false
    UseLoopCountLimit=false
   )
  ]
  particleSpawnScript [
   &InitializeParticle(
    Color=&MaskLinearColorBySpawnGroup(
     Masked Linear Color=&SelectLinearColorFromArray(
      Array Sampling Mode=::NewEnumerator0
      Direct Array Index=?
      Color Selection Array=DataInterface() {propertyValues {ColorData [<rgba 0.5576 0.315 1 1>, <rgba 1 0.5574 0.3125 1>, <rgba 0.3336 0.7227 1 1>, <rgba 1 0.2901 0.5554 1>]}}
      Random Seed=0
     )
     Passthrough Linear Color=<rgba 1.5 0.5539 0.0096 1>
     Spawn Group Mask=6
    )
    Lifetime=&RandomRangeFloat(Minimum=1.53, Maximum=2.16)
    Mass Max=5
    Mass Min=3
    Mass Mode=::NewEnumerator3
    Mesh Scale Mode=::NewEnumerator3
    Mesh Uniform Scale=&Multiply_Float(
     A=&RandomRangeFloat(Minimum=1.25, Maximum=1.55)
     B=&SelectFloatFromArray(
      Direct Array Index=&SpawnGroup()
      Float Selection Array=DataInterface() {propertyValues {FloatData [0.87,1.25,0.87,1.25]}}
     )
    )
    UsePositionOffset=false
    Write Material Random=true
   )
   &ShapeLocation(
    MaskSpawnGroup=false
    Non Uniform Scale=<1 1 1.5>
    Offset Coordinate Space=::World
    Offset Mode=::NewEnumerator0
    Override Seed=false
    Shape Origin=$User.ImpactLocation
    Sphere Radius=5
    UseOwnerScale=true
   )
   &AddVelocity(
    Velocity=&RandomRangeVector(Minimum=<0 0 350>, Maximum=<0 0 500>)
    Velocity Speed Scale=&MakeCustomFloatFromBool(Bool=Linked() {linkedVariable {name=User.Headshot?, type=NiagaraBool, defaultValue=~, description=~}}, True Float=1.5, False Float=1)
   ) {module=AddVelocity002}
   &AddVelocity(
    Cone Axis=&Multiply_VectorByFloat(Vector=$User.LinearForceVector, Float=-1)
    Inner Cone Angle=25
    Module.UseConeFalloff=false
    Module.UseVelDistribution=false
    Override Seed=false
    Velocity Mode=::NewEnumerator2
    Velocity Speed=&RandomRangeFloat(Minimum=550, Maximum=1100)
   )
   &ApplyInitialForces(Positional Force Warmup Time=0)
   &InitialMeshOrientation(
    Rotation=&RandomRangeVector(Minimum=<0 0 0>, Maximum=<1 1 1>)
    Use Orientation Vector=true
    Use Rotation Vector=true
   )
   &CalculateSizeAndRotationalInertiaByMass(
    Mass Modulation:NiagaraFloat=&SelectFloatFromArray(
     Direct Array Index=&SpawnGroup()
     Float Selection Array=DataInterface() {propertyValues {FloatData [20,50,20,50]}}
    )
    Module.Manually Enter Density=false
   )
   &DynamicMaterialParameters(
    Index 0 Param 2=1
    Module.Index0Param0Write=true
    Module.Index0Param1Write=true
    Module.Index0Param2Write=true
    Module.Index0Param3Write=false
   )
  ]
  particleUpdateScript [
   &ParticleState()
   &Timeline(
    Initial Start Time=0.3
    Play=&SetBoolByFloatComparison(A=$Particles.Age, B=0.125, Comparison Type=::NewEnumerator6)
   ) {module=Timeline001}
   &CurlNoiseForce(Noise Frequency=35, Noise Strength=&RandomRangeFloat(Minimum=450, Maximum=800), Pan Noise=false)
   &Drag(
    Drag=&Multiply_Float(
     A=&RandomRangeFloat(Minimum=7, Maximum=8)
     B=&FloatFromCurve(FloatCurve=DataInterface() {propertyValues {Curve {keys [
          {interpMode=RCIM_Cubic, leaveTangent=-4.6764, tangentMode=RCTM_Break, time=0, value=1}
          {interpMode=RCIM_Cubic, tangentMode=RCTM_Break, time=0.35, value=0}
         ]}}})
    )
    Use Linear Drag=true
    Use Rotational Drag=true
   )
   &GravityForce(Gravity=<0 0 -1100>)
   &Collision(
    Age Colliding Particles=false
    Control Roll On Collision=true
    Friction:NiagaraFloat=&RandomRangeFloat(Minimum=0.2, Maximum=0.65)
    GPU Collision Type=::NewEnumerator2
    Max Number Of Collisions=5
    Maximum Penetration Correction Distance=3
    Mesh Dimensions:Vector3f=$Particles.Scale
    Method for Calculating Particle Radius:ENiagaraCalculateRadiusOptions=::NewEnumerator2
    Module.Control Rest State Via Surface Normal=false
    Module.EnableMaxCollisionCount=true
    Module.Randomize Collision Normal Vector=true
    Percentage of Penetration Before Rest=0.2
    Radius Calculation Type=::NewEnumerator1
    Rest State Time Range=1
    Restitution=&RandomRangeFloat(Minimum=0.19, Maximum=0.35)
    ScaleRateOfRotation=true
   ) {module=Collision001}
   &Timeline(Instantaneous Rewind=true, Play=$Output.Collision001.CollisionValid, Rewind Rate=0.1, Timeline End=0.25) {enabled=false, module=Timeline002}
   &SolveForcesAndVelocity(Clamp Velocity=false, Limit Acceleration=false) {module=SolveForcesAndVelocity001}
   &ScaleColor(
    Scale Alpha=&FloatFromCurve(FloatCurve=DataInterface() {propertyValues {Curve {keys [
         {tangentMode=RCTM_User, time=0, value=1}
         {arriveTangent=-0.0405, interpMode=RCIM_Cubic, leaveTangent=-0.0405, tangentMode=RCTM_User, time=0.25, value=1}
         {arriveTangent=-3.3202, interpMode=RCIM_Cubic, leaveTangent=-3.3202, tangentMode=RCTM_User, time=1, value=0}
        ]}}})
    ScaleA=true
    ScaleRGB=false
   )
   &self:CubeCollision_01(
    NumberOfCollisions:NiagaraInt32=$Particles.Number Of Collisions
    Rate Of Change=0.25
    SurfaceNormal:Vector3f=$Output.Collision001.CombinedCollisionNormal
   ) {module=CubeCollision_01}
   &DynamicMaterialParameters(
    Index 0 Param 1=&Multiply_Float(
     A=1
     B=&FloatFromCurve(FloatCurve=DataInterface() {propertyValues {Curve {keys [
          {interpMode=RCIM_Cubic, leaveTangent=-8.6759, tangentMode=RCTM_Break, time=0, value=1}
          {interpMode=RCIM_Cubic, tangentMode=RCTM_Break, time=0.35, value=0}
         ]}}})
    )
    Index 0 Param 2=&FloatFromCurve(FloatCurve=DataInterface() {propertyValues {Curve {keys [
         {interpMode=RCIM_Cubic, tangentMode=RCTM_Break, time=0, value=1}
         {arriveTangent=-2.2159, interpMode=RCIM_Cubic, leaveTangent=-0.0891, tangentMode=RCTM_Break, time=1, value=0}
        ]}}})
    Index 0 Param 3=&Multiply_Float(
     A=5
     B=&FloatFromCurve(
      FloatCurve=DataInterface() {propertyValues {Curve {keys [
          {arriveTangent=9.0725, interpMode=RCIM_Cubic, leaveTangent=9.0725, tangentMode=RCTM_User, time=0, value=0.2}
          {arriveTangent=-2.5249, interpMode=RCIM_Cubic, leaveTangent=-2.5249, tangentMode=RCTM_User, time=0.1408, value=1.5}
          {arriveTangent=-0.0405, interpMode=RCIM_Cubic, leaveTangent=-0.0405, tangentMode=RCTM_User, time=1, value=0.1}
         ]}}}
      CurveIndex=$Particles.Timeline002.NormalizedTime
     )
    )
    Index 0 Param 4=&FloatFromCurve(FloatCurve=DataInterface() {propertyValues {Curve {keys [
         {arriveTangent=2.4654, interpMode=RCIM_Cubic, leaveTangent=2.4654, tangentMode=RCTM_User, time=0, value=0}
         {interpMode=RCIM_Cubic, tangentMode=RCTM_User, time=1, value=1}
        ]}}})
    Module.Index0Param0Write=false
    Module.Index0Param1Write=true
    Module.Index0Param2Write=false
    Module.Index0Param3Write=true
   ) {module=DynamicMaterialParameters001}
  ]
  rendererClasses [&NiagaraMeshRendererProperties]
  renderers [{rendererClass=&NiagaraMeshRendererProperties, rendererIndex=0}]
 }
 ImpactSquib_Med_High {
  simTarget=GPUComputeSim
  emitterSpawnScript [/SetVariables(Emitter.IsHeadshot:NiagaraBool=Linked() {linkedVariable {name=User.Headshot?, type=NiagaraBool, defaultValue=~, description=~}}) {isSetParameters=true, module=SetVariables_#2}]
  emitterUpdateScript [
   &EmitterState(Life Cycle Mode=::NewEnumerator1, Loop Behavior=::NewEnumerator1, Loop Duration=0.2, UseLoopDelay=false)
   &SpawnBurst_Instantaneous(
    Spawn Count=&RandomRangeInt(Minimum=4, Maximum=6)
    Spawn Group:NiagaraInt32=&MakeCustomIntFromBool(Bool=Linked() {linkedVariable {name=User.Headshot?, type=NiagaraBool, defaultValue=~, description=~}}, True INT=3, False INT=1)
    Use Spawn Probability=false
    UseLoopCountLimit=false
   ) {module=SpawnBurst_Instantaneous001}
   &SpawnBurst_Instantaneous(
    Spawn Count=&RandomRangeInt(Minimum=13, Maximum=17)
    Spawn Group:NiagaraInt32=&MakeCustomIntFromBool(Bool=Linked() {linkedVariable {name=User.Headshot?, type=NiagaraBool, defaultValue=~, description=~}}, True INT=2, False INT=0)
    Use Spawn Probability=false
    UseLoopCountLimit=false
   )
  ]
  particleSpawnScript [
   &InitializeParticle(
    Color=&MaskLinearColorBySpawnGroup(
     Masked Linear Color=&SelectLinearColorFromArray(
      Array Sampling Mode=::NewEnumerator0
      Direct Array Index=?
      Color Selection Array=DataInterface() {propertyValues {ColorData [<rgba 0.5576 0.315 1 1>, <rgba 1 0.5574 0.3125 1>, <rgba 0.3336 0.7227 1 1>, <rgba 1 0.2901 0.5554 1>]}}
      Random Seed=0
     )
     Passthrough Linear Color=<rgba 1.5 0.5539 0.0096 1>
     Spawn Group Mask=6
    )
    Lifetime=&RandomRangeFloat(Minimum=0.75, Maximum=1.4)
    Mass Max=5
    Mass Min=3
    Mass Mode=::NewEnumerator3
    Mesh Scale Mode=::NewEnumerator3
    Mesh Uniform Scale=&Multiply_Float(
     A=&RandomRangeFloat(Minimum=1.25, Maximum=1.55)
     B=&SelectFloatFromArray(
      Direct Array Index=&SpawnGroup()
      Float Selection Array=DataInterface() {propertyValues {FloatData [0.87,1.25,0.87,1.25]}}
     )
    )
    UsePositionOffset=false
    Write Material Random=true
   )
   &ShapeLocation(
    MaskSpawnGroup=false
    Non Uniform Scale=<1 1 1.5>
    Offset Coordinate Space=::World
    Offset Mode=::NewEnumerator0
    Override Seed=false
    Shape Origin=$User.ImpactLocation
    Sphere Radius=5
    UseOwnerScale=true
   )
   &AddVelocity(
    Velocity=&RandomRangeVector(Minimum=<0 0 350>, Maximum=<0 0 500>)
    Velocity Speed Scale=&MakeCustomFloatFromBool(Bool=Linked() {linkedVariable {name=User.Headshot?, type=NiagaraBool, defaultValue=~, description=~}}, True Float=1.5, False Float=1)
   ) {module=AddVelocity002}
   &AddVelocity(
    Cone Axis=&Multiply_VectorByFloat(Vector=$User.LinearForceVector, Float=-1)
    Inner Cone Angle=25
    Module.UseConeFalloff=false
    Module.UseVelDistribution=false
    Override Seed=false
    Velocity Mode=::NewEnumerator2
    Velocity Speed=&RandomRangeFloat(Minimum=550, Maximum=1100)
   )
   &ApplyInitialForces(Positional Force Warmup Time=0)
   &InitialMeshOrientation(
    Rotation=&RandomRangeVector(Minimum=<0 0 0>, Maximum=<1 1 1>)
    Use Orientation Vector=true
    Use Rotation Vector=true
   )
   &CalculateSizeAndRotationalInertiaByMass(
    Mass Modulation:NiagaraFloat=&SelectFloatFromArray(
     Direct Array Index=&SpawnGroup()
     Float Selection Array=DataInterface() {propertyValues {FloatData [20,50,20,50]}}
    )
    Module.Manually Enter Density=false
   )
   &DynamicMaterialParameters(
    Index 0 Param 2=1
    Module.Index0Param0Write=true
    Module.Index0Param1Write=true
    Module.Index0Param2Write=true
    Module.Index0Param3Write=false
   )
  ]
  particleUpdateScript [
   &ParticleState()
   &Timeline(
    Initial Start Time=0.3
    Play=&SetBoolByFloatComparison(A=$Particles.Age, B=0.125, Comparison Type=::NewEnumerator6)
   ) {module=Timeline001}
   &CurlNoiseForce(Noise Frequency=35, Noise Strength=&RandomRangeFloat(Minimum=450, Maximum=800), Pan Noise=false)
   &Drag(
    Drag=&Multiply_Float(
     A=&RandomRangeFloat(Minimum=7, Maximum=8)
     B=&FloatFromCurve(FloatCurve=DataInterface() {propertyValues {Curve {keys [
          {interpMode=RCIM_Cubic, leaveTangent=-4.6764, tangentMode=RCTM_Break, time=0, value=1}
          {interpMode=RCIM_Cubic, tangentMode=RCTM_Break, time=0.35, value=0}
         ]}}})
    )
    Use Linear Drag=true
    Use Rotational Drag=true
   )
   &GravityForce(Gravity=<0 0 -1100>)
   &Collision(
    Age Colliding Particles=false
    Analytical Collision Plane Position 1:Vector3f=<0 0 0>
    Control Roll On Collision=true
    Friction:NiagaraFloat=&RandomRangeFloat(Minimum=0.2, Maximum=0.65)
    GPU Collision Type=::NewEnumerator3
    Max Number Of Collisions=5
    Maximum Penetration Correction Distance=3
    Mesh Dimensions:Vector3f=$Particles.Scale
    Method for Calculating Particle Radius:ENiagaraCalculateRadiusOptions=::NewEnumerator2
    Module.Control Rest State Via Surface Normal=false
    Module.EnableMaxCollisionCount=true
    Module.Randomize Collision Normal Vector=true
    Percentage of Penetration Before Rest=0.2
    Radius Calculation Type=::NewEnumerator1
    Rest State Time Range=1
    Restitution=&RandomRangeFloat(Minimum=0.19, Maximum=0.35)
    ScaleRateOfRotation=true
   ) {module=Collision001}
   &Timeline(Instantaneous Rewind=true, Play=$Output.Collision001.CollisionValid, Rewind Rate=0.1, Timeline End=0.25) {enabled=false, module=Timeline002}
   &SolveForcesAndVelocity(Clamp Velocity=false, Limit Acceleration=false) {module=SolveForcesAndVelocity001}
   &ScaleColor(
    Scale Alpha=&FloatFromCurve(FloatCurve=DataInterface() {propertyValues {Curve {keys [
         {tangentMode=RCTM_User, time=0, value=1}
         {arriveTangent=-0.0405, interpMode=RCIM_Cubic, leaveTangent=-0.0405, tangentMode=RCTM_User, time=0.25, value=1}
         {arriveTangent=-3.3202, interpMode=RCIM_Cubic, leaveTangent=-3.3202, tangentMode=RCTM_User, time=1, value=0}
        ]}}})
    ScaleA=true
    ScaleRGB=false
   )
   &self:CubeCollision_01(
    NumberOfCollisions:NiagaraInt32=$Particles.Number Of Collisions
    Rate Of Change=0.25
    SurfaceNormal:Vector3f=$Output.Collision001.CombinedCollisionNormal
   ) {module=CubeCollision_01}
   &DynamicMaterialParameters(
    Index 0 Param 1=&Multiply_Float(
     A=1
     B=&FloatFromCurve(FloatCurve=DataInterface() {propertyValues {Curve {keys [
          {interpMode=RCIM_Cubic, leaveTangent=-8.6759, tangentMode=RCTM_Break, time=0, value=1}
          {interpMode=RCIM_Cubic, tangentMode=RCTM_Break, time=0.35, value=0}
         ]}}})
    )
    Index 0 Param 2=&FloatFromCurve(FloatCurve=DataInterface() {propertyValues {Curve {keys [
         {interpMode=RCIM_Cubic, tangentMode=RCTM_Break, time=0, value=1}
         {arriveTangent=-2.2159, interpMode=RCIM_Cubic, leaveTangent=-0.0891, tangentMode=RCTM_Break, time=1, value=0}
        ]}}})
    Index 0 Param 3=&Multiply_Float(
     A=5
     B=&FloatFromCurve(
      FloatCurve=DataInterface() {propertyValues {Curve {keys [
          {arriveTangent=9.0725, interpMode=RCIM_Cubic, leaveTangent=9.0725, tangentMode=RCTM_User, time=0, value=0.2}
          {arriveTangent=-2.5249, interpMode=RCIM_Cubic, leaveTangent=-2.5249, tangentMode=RCTM_User, time=0.1408, value=1.5}
          {arriveTangent=-0.0405, interpMode=RCIM_Cubic, leaveTangent=-0.0405, tangentMode=RCTM_User, time=1, value=0.1}
         ]}}}
      CurveIndex=$Particles.Timeline002.NormalizedTime
     )
    )
    Index 0 Param 4=&FloatFromCurve(FloatCurve=DataInterface() {propertyValues {Curve {keys [
         {arriveTangent=2.4654, interpMode=RCIM_Cubic, leaveTangent=2.4654, tangentMode=RCTM_User, time=0, value=0}
         {interpMode=RCIM_Cubic, tangentMode=RCTM_User, time=1, value=1}
        ]}}})
    Module.Index0Param0Write=false
    Module.Index0Param1Write=true
    Module.Index0Param2Write=false
    Module.Index0Param3Write=true
   ) {module=DynamicMaterialParameters001}
  ]
  rendererClasses [&NiagaraMeshRendererProperties]
  renderers [{rendererClass=&NiagaraMeshRendererProperties, rendererIndex=0}]
 }
 Impact_Flare_and_Corona {
  emitterUpdateScript [
   &EmitterState(Life Cycle Mode=::NewEnumerator1, Loop Behavior=::NewEnumerator1, Loop Delay=0.175, Loop Duration=0.5, UseLoopDelay=true)
   &SpawnBurst_Instantaneous(Spawn Count=1, Use Spawn Probability=false)
  ]
  particleSpawnScript [
   &InitializeParticle(
    Color=<rgba 8 4.2 0.4 1>
    Lifetime=0.3
    Sprite Rotation Mode=::NewEnumerator2
    Sprite Size Mode=::NewEnumerator3
    Uniform Sprite Size=300
    UsePositionOffset=false
    Write Material Random=true
   ) {module=InitializeParticle001}
   &ShapeLocation(
    MaskSpawnGroup=false
    Non Uniform Scale=<1 1 1.5>
    Offset Coordinate Space=::World
    Offset Mode=::NewEnumerator0
    Override Seed=false
    Shape Origin=$User.ImpactLocation
    Sphere Radius=1
    UseOwnerScale=true
   )
   &CameraOffset(Camera Offset Amount=-35)
  ]
  particleUpdateScript [
   &ParticleState()
   &DynamicMaterialParameters(
    Index 0 Param 1=&FloatFromCurve(FloatCurve=DataInterface() {propertyValues {Curve {keys [<at 0 1 RCIM_Cubic>,<at 0.8 0 RCIM_Cubic>]}}})
    Index 0 Param 2=2
    Index 0 Param 3=1
    Index 0 Param 4=1
    Index 1 Param 1=&FloatFromCurve(FloatCurve=DataInterface() {propertyValues {Curve {keys [
         {arriveTangent=-2.3163, interpMode=RCIM_Cubic, time=0, value=2}
         {arriveTangent=-1.9019, interpMode=RCIM_Cubic, leaveTangent=-1.9019, tangentMode=RCTM_User, time=1, value=0.4}
        ]}}})
    Index 1 Param 2=0.4
    Index 1 Param 3=0.6
    Index 1 Param 4=1
    Module.Param0WriteEnabled=true
    Module.Param0WriteEnabled1=false
    Module.Param1WriteEnabled=true
    Module.Param1WriteEnabled1=true
    Module.Param2WriteEnabled=true
    Module.Param2WriteEnabled1=true
    Module.Param3WriteEnabled=true
    Module.Param3WriteEnabled1=true
    Write Parameter Index 1=true
   )
   &ScaleColor(
    Scale Alpha=&FloatFromCurve(FloatCurve=DataInterface() {propertyValues {Curve {keys [
         {interpMode=RCIM_Cubic, tangentMode=RCTM_User, time=0, value=1}
         {arriveTangent=-0.0145, interpMode=RCIM_Cubic, leaveTangent=-0.0145, tangentMode=RCTM_User, time=1, value=0}
        ]}}})
    Scale RGB=&MakeVectorFromLinearColorRGB(Color=<rgba 1 1 1 1>)
    ScaleA=true
    ScaleRGB=true
   )
   &ScaleSpriteSize(Scale Factor=&Vector2DFromFloat(Value=&FloatFromCurve(FloatCurve=DataInterface() {propertyValues {Curve {
         defaultValue=0.5
         keys [
          {arriveTangent=0.9354, interpMode=RCIM_Cubic, leaveTangent=0.9354, tangentMode=RCTM_User, time=0, value=0.6}
          {interpMode=RCIM_Cubic, tangentMode=RCTM_User, time=1, value=1}
         ]
        }}})))
   /SetVariables(Particles.Halo Sprite Color:LinearColor=&MakeLinearColorFromVectorAndFloat(
     "Vector (RGB)"=&MakeVectorFromLinearColorRGB(Color=&LerpLinearColors(
       StartColor=<rgba 2 0.95 0.2 1>
       EndColor=<rgba 0.5 0.1943 0 1>
       LerpFactor=&FloatFromCurve(FloatCurve=DataInterface() {propertyValues {Curve {keys [<at 0.35 0 RCIM_Cubic>,<at 1 1 RCIM_Cubic>]}}})
      ))
     "Float (Alpha)"=&FloatFromCurve(FloatCurve=DataInterface() {propertyValues {Curve {
         defaultValue=1
         keys [
          {interpMode=RCIM_Cubic, tangentMode=RCTM_User, time=0, value=1}
          {arriveTangent=-1.1174, interpMode=RCIM_Cubic, leaveTangent=-1.1174, tangentMode=RCTM_User, time=1, value=0}
         ]
        }}})
    )) {isSetParameters=true}
   /SetVariables(Particles.Halo Sprite Size:Vector2f=&Multiply_Vector2D_ByFloat(Float=0.6, Vector2D=$Particles.SpriteSize)) {isSetParameters=true}
   /SetVariables(Particles.Iridescent Sprite Color:LinearColor=&MakeLinearColorFromVectorAndFloat(
     "Vector (RGB)"=&MakeVectorFromLinearColorRGB(Color=&LerpLinearColors(
       StartColor=<rgba 2 1.1076 0.5833 1>
       EndColor=<rgba 1 0.3887 0 1>
       LerpFactor=&FloatFromCurve(FloatCurve=DataInterface() {propertyValues {Curve {keys [<at 0.35 0 RCIM_Cubic>,<at 1 1 RCIM_Cubic>]}}})
      ))
     "Float (Alpha)"=&FloatFromCurve(FloatCurve=DataInterface() {propertyValues {Curve {
         defaultValue=1
         keys [
          {interpMode=RCIM_Cubic, tangentMode=RCTM_User, time=0, value=1}
          {arriveTangent=-1.5378, interpMode=RCIM_Cubic, leaveTangent=-1.5378, tangentMode=RCTM_User, time=1, value=0}
         ]
        }}})
    )) {isSetParameters=true}
   /SetVariables(Particles.Iridescent Sprite Size:Vector2f=&Multiply_Vector2D_ByFloat(Float=1, Vector2D=$Particles.SpriteSize)) {isSetParameters=true}
  ]
  rendererClasses [&NiagaraSpriteRendererProperties]
  renderers [
   {rendererClass=&NiagaraSpriteRendererProperties}
   {rendererClass=&NiagaraSpriteRendererProperties, rendererIndex=1}
   {rendererClass=&NiagaraSpriteRendererProperties, rendererIndex=2}
  ]
 }
 Sparks {
  emitterUpdateScript [
   &EmitterState(Life Cycle Mode=::NewEnumerator1, Loop Behavior=::NewEnumerator1, Loop Duration=0.275, UseLoopDelay=false)
   &SpawnRate(SpawnRate=500, Use Spawn Probability=false)
  ]
  particleSpawnScript [
   &InitializeParticle(
    Color=&UniformRangedLinearColor(Minimum=<rgba 10 2.3834 0 1>, Maximum=<rgba 30 2.3601 0 1>)
    Lifetime=&UniformRangedFloat(Minimum=0.25, Maximum=0.5)
    Mass=5
    Mass Mode=::NewEnumerator1
    Sprite Size Mode=::NewEnumerator1
    Uniform Sprite Size Max=5
    Uniform Sprite Size Min=2.5
    UsePositionOffset=false
    Write Material Random=true
   ) {module=InitializeParticle001}
   &SkeletalMeshLocation(Inherit Velocity=false, MaskSpawnGroup=false, Module.Override Randomness=false, Skeletal Mesh=$User.SkeletalMesh) {enabled=false}
   &SphereLocation(
    MaskSpawnGroup=false
    Override Seed=false
    Sphere Coordinate Space=::World
    Sphere Origin:NiagaraPosition=$User.ImpactLocation
    Sphere Radius=15
    Surface Only=false
    UseNonUniformScale=false
    UseOffset=false
    UseSphereOrientation=false
   )
   &CameraOffset(Camera Offset Amount=-25)
   &PointForce(
    Force Falloff Distance=40
    Force Origin:NiagaraPosition=$User.ImpactLocation
    Force Strength=&RandomRangeFloat(Minimum=490, Maximum=1470)
    Linear Force Vector:Vector3f=&MakeCustomVectorFromBool(
     Bool=Linked() {linkedVariable {name=User.Headshot?, type=NiagaraBool, defaultValue=~, description=~}}
     True Vector=&Add_Vector(A=&Multiply_VectorByFloat(Vector=$User.LinearForceVector, Float=2), B=<0 0 3.25>)
     False Vector=&Multiply_VectorByFloat(
      Vector=&Multiply_Vector(A=&RandomRangeVector(Minimum=<0.75 0.75 1>, Maximum=<1.15 1.15 1.15>), B=$User.LinearForceVector)
      Float=&RandomRangeFloat(Minimum=10, Maximum=15)
     )
    )
    Module.Use Falloff=true
    Module.UseLinearForce=true
   )
   &AddVelocity(
    Velocity=&RandomRangeVector(Minimum=<0 0 120>, Maximum=<0 0 350>)
    Velocity Speed Scale=&MakeCustomFloatFromBool(Bool=Linked() {linkedVariable {name=User.Headshot?, type=NiagaraBool, defaultValue=~, description=~}}, True Float=1, False Float=0.25)
   ) {module=AddVelocity001}
   &AddVelocity(
    Cone Axis=&MakeCustomVectorFromBool(
     Bool=&SetBoolByFloatComparison(A=&VectorLength(Vector=$User.PlayerVelocity, Fallback Vector=<0 0 0>), B=0)
     True Vector=$User.PlayerVelocity
     False Vector=$User.LinearForceVector
    )
    Inner Cone Angle=20
    Module.UseConeFalloff=false
    Module.UseVelDistribution=false
    Override Seed=false
    Velocity Mode=::NewEnumerator2
    Velocity Speed=&Multiply_Float(
     A=&RandomRangeFloat(Minimum=300, Maximum=555)
     B=&MakeCustomFloatFromBool(Bool=Linked() {linkedVariable {name=User.Headshot?, type=NiagaraBool, defaultValue=~, description=~}}, True Float=1.6, False Float=1.3)
    )
   ) {module=AddVelocity002}
   &ApplyInitialForces()
  ]
  particleUpdateScript [
   &ParticleState()
   &GravityForce(Gravity=<0 0 -1200>)
   &Drag(
    Drag=&Multiply_Float(
     A=&RandomRangeFloat(Minimum=3.5, Maximum=8)
     B=&FloatFromCurve(FloatCurve=DataInterface() {propertyValues {Curve {keys [
          {interpMode=RCIM_Cubic, leaveTangent=-4.7225, tangentMode=RCTM_Break, time=0, value=1}
          {interpMode=RCIM_Cubic, tangentMode=RCTM_Break, time=0.5, value=0}
         ]}}})
    )
    Use Linear Drag=true
    Use Rotational Drag=true
   )
   &SolveForcesAndVelocity(Clamp Velocity=false, Limit Acceleration=false)
   &ScaleSpriteSizeBySpeed(Max Scale Factor=<0.5 2.5>)
   &ScaleSpriteSize(Scale Factor=&Vector2DFromFloat(Value=&FloatFromCurve(FloatCurve=DataInterface() {propertyValues {Curve {keys [<at 0.5 1 RCIM_Cubic>, {arriveTangent=-3.0707, interpMode=RCIM_Cubic, leaveTangent=-3.0707, tangentMode=RCTM_User, time=1, value=0}]}}})))
   /Niagara/Modules/Update/Color/ScaleColorBySpeed(A Scale Min=0.1, RGB Scale Max=5, RGB Scale Min=1)
  ]
  rendererClasses [&NiagaraSpriteRendererProperties]
  renderers [{rendererClass=&NiagaraSpriteRendererProperties}]
 }
 Splat_Mesh {
  emitterUpdateScript [
   &EmitterState(Life Cycle Mode=::NewEnumerator1, Loop Behavior=::NewEnumerator1, Loop Duration=1, UseLoopDelay=false)
   &SpawnBurst_Instantaneous(Spawn Count=1, Use Spawn Probability=false)
  ]
  particleSpawnScript [
   &InitializeParticle(
    Color=<rgba 5 3.1304 0 1>
    Lifetime=0.35
    Mesh Scale Max=<0.35 0.35 0.35>
    Mesh Scale Min=<0.2 0.2 0.2>
    Mesh Scale Mode=::NewEnumerator2
    UsePositionOffset=false
    Write Material Random=true
   ) {module=InitializeParticle001}
   &ShapeLocation(
    MaskSpawnGroup=false
    Non Uniform Scale=<1 1 1.5>
    Offset Coordinate Space=::World
    Offset Mode=::NewEnumerator0
    Override Seed=false
    Shape Origin=$User.ImpactLocation
    Sphere Radius=1
    UseOwnerScale=true
   )
   &DynamicMaterialParameters(
    Index 0 Param 3=1
    Module.Param0WriteEnabled=false
    Module.Param1WriteEnabled=true
    Module.Param2WriteEnabled=true
    Module.Param3WriteEnabled=true
   ) {module=DynamicMaterialParameters001}
   &InitialMeshOrientation(
    Rotation=&RandomRangeVector(Minimum=<0 0 0>, Maximum=<1 1 1>)
    Rotation Coordinate Space=::NewEnumerator1
    Use Orientation Vector=true
    Use Rotation Vector=true
   )
   &CameraOffset(Camera Offset Amount=10)
  ]
  particleUpdateScript [
   &ParticleState()
   &DynamicMaterialParameters(
    Index 0 Param 1=&FloatFromCurve(FloatCurve=DataInterface() {propertyValues {Curve {keys [<at 0.56 0>,<at 1 1>]}}})
    Index 0 Param 2=1
    Module.Param0WriteEnabled=false
    Module.Param1WriteEnabled=true
    Module.Param2WriteEnabled=false
    Module.Param3WriteEnabled=false
   )
   &ScaleMeshSize(Scale Factor=&VectorFromFloat(Value=&FloatFromCurve(FloatCurve=DataInterface() {propertyValues {Curve {keys [<at 0 0.5>,<at 1 1>]}}}, Scale Curve=2)))
   &ScaleColor(
    Color Value To Scale=$Particles.ImpactFlareColor
    Scale Alpha=&FloatFromCurve(FloatCurve=DataInterface() {propertyValues {Curve {keys [<at 0 1>,<at 1 0>]}}})
    Scale RGB=&MakeVectorFromLinearColorRGB(Color=&ColorFromCurve(ColorCurve=DataInterface() {propertyValues {
        AlphaCurve {keys [<at 0 1 RCIM_Cubic>,<at 1 1 RCIM_Cubic>]}
        BlueCurve {keys [<at 0 0>,<at 1 0>]}
        GreenCurve {keys [<at 0 0.5217>,<at 1 2.6087>]}
        RedCurve {keys [<at 0 1>,<at 1 50>]}
       }}))
    ScaleA=true
    ScaleRGB=false
   ) {module=ScaleColor004}
   &ScaleColor(
    Scale Alpha=&FloatFromCurve(FloatCurve=DataInterface() {propertyValues {Curve {keys [<at 0 1>,<at 1 0>]}}})
    Scale RGB=&MakeVectorFromLinearColorRGB(Color=&ColorFromCurve(ColorCurve=DataInterface() {propertyValues {
        AlphaCurve {keys [<at 0 1 RCIM_Cubic>,<at 1 1 RCIM_Cubic>]}
        BlueCurve {keys [<at 0 0>,<at 1 0>]}
        GreenCurve {keys [<at 0 0.5217>,<at 1 2.6087>]}
        RedCurve {keys [<at 0 1>,<at 1 50>]}
       }}))
    ScaleA=false
    ScaleRGB=true
   )
   &ScaleColor(
    Scale RGB=&VectorFromFloat(Value=&FloatFromCurve(FloatCurve=DataInterface() {propertyValues {Curve {keys [<at 0 10>,<at 0.2 3>,<at 1 1>]}}}))
    ScaleA=true
    ScaleRGB=true
   ) {module=ScaleColor001}
   &ScaleColor(
    Scale RGB=&VectorFromFloat(Value=&FNTimeOfDayScalar(Night Scale=0.7, Morning Scale=0.5, Evening Scale=0.5))
    ScaleA=true
    ScaleRGB=true
   ) {module=ScaleColor002}
  ]
  rendererClasses [&NiagaraMeshRendererProperties]
  renderers [{rendererClass=&NiagaraMeshRendererProperties, rendererIndex=0}]
 }
 Squares__Low_End_Only_ {
  emitterUpdateScript [
   &EmitterState(Life Cycle Mode=::NewEnumerator1, Loop Behavior=::NewEnumerator1, Loop Duration=0.275, UseLoopDelay=false)
   &SpawnRate(SpawnRate=1800, Use Spawn Probability=false)
  ]
  particleSpawnScript [
   &InitializeParticle(
    Color=&UniformRangedLinearColor(Minimum=<rgba 2 0.5217 0 1>, Maximum=<rgba 5 1.5652 0 1>)
    Lifetime=&UniformRangedFloat(Minimum=0.25, Maximum=0.7)
    Mass=5
    Mass Mode=::NewEnumerator1
    Sprite Size Mode=::NewEnumerator1
    Uniform Sprite Size Min=4
    UsePositionOffset=false
    Write Material Random=true
   ) {module=InitializeParticle001}
   &SkeletalMeshLocation(Inherit Velocity=false, MaskSpawnGroup=false, Module.Override Randomness=false, Skeletal Mesh=$User.SkeletalMesh)
   &SphereLocation(
    MaskSpawnGroup=false
    Override Seed=false
    Sphere Radius=5
    Surface Only=false
    UseNonUniformScale=false
    UseOffset=false
    UseSphereOrientation=false
   )
   &CameraOffset(Camera Offset Amount=50)
   &PointForce(
    Force Falloff Distance=60
    Force Origin:NiagaraPosition=$User.ImpactLocation
    Force Strength=&RandomRangeFloat(Minimum=490, Maximum=1470)
    Linear Force Vector:Vector3f=&MakeCustomVectorFromBool(
     Bool=Linked() {linkedVariable {name=User.Headshot?, type=NiagaraBool, defaultValue=~, description=~}}
     True Vector=&Add_Vector(A=&Multiply_VectorByFloat(Vector=$User.LinearForceVector, Float=2), B=<0 0 3.25>)
     False Vector=&Multiply_VectorByFloat(
      Vector=&Multiply_Vector(A=&RandomRangeVector(Minimum=<0.75 0.75 0.75>, Maximum=<1.15 1.15 1.15>), B=$User.LinearForceVector)
      Float=15
     )
    )
    Module.Use Falloff=true
    Module.UseLinearForce=true
   )
   &AddVelocity(
    Velocity=&RandomRangeVector(Minimum=<0 0 120>, Maximum=<0 0 350>)
    Velocity Speed Scale=&MakeCustomFloatFromBool(Bool=Linked() {linkedVariable {name=User.Headshot?, type=NiagaraBool, defaultValue=~, description=~}}, True Float=1, False Float=0.25)
   ) {module=AddVelocity001}
   &AddVelocity(
    Cone Axis=$User.LinearForceVector
    Inner Cone Angle=20
    Module.UseConeFalloff=false
    Module.UseVelDistribution=false
    Override Seed=false
    Velocity Mode=::NewEnumerator2
    Velocity Speed=&Multiply_Float(
     A=&RandomRangeFloat(Minimum=300, Maximum=555)
     B=&MakeCustomFloatFromBool(Bool=Linked() {linkedVariable {name=User.Headshot?, type=NiagaraBool, defaultValue=~, description=~}}, True Float=1.6, False Float=1.3)
    )
   ) {module=AddVelocity002}
   &ApplyInitialForces()
  ]
  particleUpdateScript [
   &ParticleState()
   &GravityForce()
   &Drag(
    Drag=&Multiply_Float(
     A=&RandomRangeFloat(Minimum=3.5, Maximum=8)
     B=&FloatFromCurve(FloatCurve=DataInterface() {propertyValues {Curve {keys [
          {interpMode=RCIM_Cubic, leaveTangent=-4.7225, tangentMode=RCTM_Break, time=0, value=1}
          {interpMode=RCIM_Cubic, tangentMode=RCTM_Break, time=0.5, value=0}
         ]}}})
    )
    Use Linear Drag=true
    Use Rotational Drag=true
   )
   &SolveForcesAndVelocity(Clamp Velocity=false, Limit Acceleration=false)
   &ScaleSpriteSize(Scale Factor=&Vector2DFromFloat(Value=&FloatFromCurve(FloatCurve=DataInterface() {propertyValues {Curve {keys [<at 0.5 1 RCIM_Cubic>, {arriveTangent=-3.0707, interpMode=RCIM_Cubic, leaveTangent=-3.0707, tangentMode=RCTM_User, time=1, value=0}]}}})))
  ]
  rendererClasses [&NiagaraSpriteRendererProperties]
  renderers [{rendererClass=&NiagaraSpriteRendererProperties}]
 }
}
summary {
 systemName=NS_Player_Teleport_Out_Athena_TacElim_GoldenGun
 emitters [
  {emitterName=Squares__Low_End_Only_, rendererClasses [&NiagaraSpriteRendererProperties]}
  {emitterName=Dynamic_Cubes, simTarget=GPUComputeSim, rendererClasses [&NiagaraMeshRendererProperties]}
  {emitterName=ImpactSquib, simTarget=GPUComputeSim, rendererClasses [&NiagaraMeshRendererProperties]}
  {emitterName=Impact_Flare_and_Corona, rendererClasses [&NiagaraSpriteRendererProperties]}
  {emitterName=Splat_Mesh, rendererClasses [&NiagaraMeshRendererProperties]}
  {emitterName=Sparks, rendererClasses [&NiagaraSpriteRendererProperties]}
  {emitterName=Flash, rendererClasses [&NiagaraMeshRendererProperties]}
  {emitterName=Dynamic_Cubes_Med_High, simTarget=GPUComputeSim, rendererClasses [&NiagaraMeshRendererProperties]}
  {emitterName=ImpactSquib_Med_High, simTarget=GPUComputeSim, rendererClasses [&NiagaraMeshRendererProperties]}
  {emitterName=Gold_Glints, rendererClasses [&NiagaraSpriteRendererProperties]}
 ]
 userVariables [
  User.Headshot? {type=NiagaraBool}
  User.ImpactLocation {type=NiagaraPosition, description="", defaultValue {}}
  User.LinearForceVector {type=Vector3f, description="", defaultValue {struct=Vector3f, value=<0.3 0.5 0.1>}}
  User.PlayerVelocity {type=Vector3f, description="", defaultValue {struct=Vector3f, value=<0 0 0>}}
  User.SkeletalMesh {
   type=NiagaraDataInterfaceSkeletalMesh
   defaultValue {
    struct=NiagaraExt_VariableValue_DataInterface
    value {dataInterface=&self:NiagaraDataInterfaceSkeletalMesh_1}
   }
  }
 ]
}
